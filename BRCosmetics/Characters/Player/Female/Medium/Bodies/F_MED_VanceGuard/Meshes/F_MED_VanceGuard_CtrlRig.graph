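# Control Rig for the VanceGuard female medium body that applies procedural corrective motion to deform bones and costume pieces.
# A shared DeformRig function runs first, then weighted parent constraints place belt pots and kneepads.
# Smoothed spherical pose readers on left thigh, clavicles and calves offset belt, bag, collar and kneepad bones.
# Remapped bone-to-bone distances rotate collar bones as back hair or head approach, faking collision without physics.
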ControlRigBlueprint /BRCosmetics/Characters/Player/Female/Medium/Bodies/F_MED_VanceGuard/Meshes/F_MED_VanceGuard_CtrlRig
graphs [
 RigVMFunctionLibrary {links [], nodes []}
 RigVMModel {nodes [
   RigUnit_BeginExecution = &RigUnit_BeginExecution()
   FN_DeformRig_FNC = RigVMFunctionReferenceNode(
    DeformRig_ Curve="(deformRig_l_upperarm_up_alpha_72_#1=False,deformRig_r_upperarm_up_alpha_73_#2=False,deformRig_l_upperarm_fr_alpha_74_#3=False,deformRig_r_upperarm_fr_alpha_75_CB51B8 ...(truncated)"
    DeformRig_Bool="(upperarm_1_#4=True,elbow_3_#5=True,wrist_7_#6=True,groin_8_#7=True,glute_17_#8=True,knee_9_40C408324CA261 ...(truncated)"
    DeformRig_Debug=(upperarm_debug_18_#4=False,wrist_debug_20_#6=False,lat_debug_26_#9=False,groinGlute_debug_27_#7=False)
    DeformRig_Value="(upperarm_up_18_#10=(Rotation=(X=0,Y=0,Z=0,W=1),Translation=(X=0,Y=0,Z=8),Scale3D=(X=0.7,Y=1,Z=1)),upperarm_fwd_fr_19_#11=(Rotation=(X=0,Y=0,Z=0,W=1),Translation=(X=0,Y=10,Z=0),Scale3D=(X=0.5,Y=1,Z= ...(truncated)"
    ExecuteContext=@RigUnit_BeginExecution.ExecutePin
   )
   > n1 = &RigUnit_SphericalPoseReader(
    ActiveRegionScaleFactors=(PositiveWidth=1,NegativeWidth=0.1,PositiveHeight=1,NegativeHeight=1)
    ActiveRegionSize=0.3
    Debug=(bDrawDebug=false,bDraw2D=false,bDrawLocalAxes=false,DebugScale=30,DebugSegments=30,DebugThickness=0.2)
    DriverAxis=(X=-1,Y=0,Z=0)
    DriverItem=(Type=Bone,Name="thigh_l")
    FalloffRegionScaleFactors=(PositiveWidth=1,NegativeWidth=0.6,PositiveHeight=1,NegativeHeight=1)
    FalloffSize=0.5
    OptionalParentItem=(Type=Bone,Name="pelvis")
    RotationOffset=(X=90,Y=0,Z=0)
   )
   n2 = &RigVMFunction_AlphaInterp(InterpSpeedDecreasing=10, InterpSpeedIncreasing=10, Scale=1, bClampResult=False, Value=@n1.OutputParam)
   > &RigUnit_ModifyTransforms(
    ItemToModify="((Item=(Type=Bone,Name=\"BASE_belt_01\"),Transform=(Rotation=(X=0,Y=0,Z=0,W=1),Translation=(X=-6,Y=0,Z=0),Scale3D=(X=1,Y=1,Z=1))),(Item=(Type=Bone,Name=\"BASE_bag\"),Transform=(Rotation=(X=0,Y=0,Z=0,W=1),Translation=(X=5,Y=0,Z=2.5),Scale3D=(X=1,Y=1,Z=1))),(Ite ...(truncated)"
    Weight=@n2.Result
   )
   > &RigUnit_ParentConstraint(
    AdvancedSettings=(InterpolationType=Average,RotationOrderForFilter=XZY)
    Child=(Type=Bone,Name="BASE_pot_1")
    Parents=((Item=(Type=Bone,Name="belt_01"),Weight=0.1007),(Item=(Type=Bone,Name="spine_02"),Weight=0.041),(Item=(Type=Bone,Name="belt_04"),Weight=0.6493),(Item=(Type=Bone,Name="pelvis"),Weight=0.2091))
    Weight=1
    bMaintainOffset=true
   )
   > &RigUnit_ParentConstraint(
    AdvancedSettings=(InterpolationType=Average,RotationOrderForFilter=XZY)
    Child=(Type=Bone,Name="BASE_pot_2")
    Parents=((Item=(Type=Bone,Name="belt_01"),Weight=0.322),(Item=(Type=Bone,Name="belt_02"),Weight=0.0214),(Item=(Type=Bone,Name="belt_04"),Weight=0.6311),(Item=(Type=Bone,Name="pelvis"),Weight=0.0254))
    Weight=1
    bMaintainOffset=true
   )
   > &RigUnit_ParentConstraint(
    AdvancedSettings=(InterpolationType=Average,RotationOrderForFilter=XZY)
    Child=(Type=Bone,Name="BASE_pot_3")
    Parents=((Item=(Type=Bone,Name="belt_01"),Weight=0.7654),(Item=(Type=Bone,Name="belt_02"),Weight=0.0966),(Item=(Type=Bone,Name="belt_04"),Weight=0.1379))
    Weight=1
    bMaintainOffset=true
   )
   n3 = &RigVMFunction_MathVectorDistance(
    B=&RigUnit_GetTransform(Item=(Type=Bone,Name="collar_r2_r_END"), bInitial=False).Transform.Translation
    A=&RigUnit_GetTransform(Item=(Type=Bone,Name="BASE_back_hair"), bInitial=False).Transform.Translation
   )
   n4 = &RigVMFunction_MathVectorDistance(
    B=&RigUnit_GetTransform(Item=(Type=Bone,Name="collar_r2_r_END"), bInitial=False).Transform.Translation
    A=&RigUnit_GetTransform(Item=(Type=Bone,Name="back_hair_03"), bInitial=False).Transform.Translation
   )
   > &RigUnit_ModifyTransforms(
    ItemToModify=((Item=(Type=Bone,Name="BASE_collar_r2_r"),Transform=(Rotation=(X=0,Y=0,Z=-0.4617,W=0.887),Translation=(X=0,Y=0,Z=0),Scale3D=(X=1,Y=1,Z=1))))
    WeightMinimum=0
    Weight=&RigVMFunction_MathDoubleMax(
     A=&RigVMFunction_MathDoubleRemap(SourceMaximum=4, SourceMinimum=6.1, TargetMaximum=1, TargetMinimum=0, Value=@n3.Result).Result
     B=&RigVMFunction_MathDoubleRemap(SourceMaximum=4, SourceMinimum=4.5, TargetMaximum=1, TargetMinimum=0, Value=@n4.Result).Result
    ).Result
   )
   n5 = &RigUnit_GetTransform(Item=(Type=Bone,Name="collar_r4_r END"), bInitial=False)
   n6 = &RigUnit_GetTransform(Item=(Type=Bone,Name="BASE_back_hair"), bInitial=False)
   > &RigUnit_ModifyTransforms(
    ItemToModify=((Item=(Type=Bone,Name="BASE_collar_r4_r"),Transform=(Rotation=(X=0,Y=0,Z=-0.3827,W=0.9239),Translation=(X=0,Y=0,Z=0),Scale3D=(X=1,Y=1,Z=1))))
    WeightMinimum=0
    Weight=&RigVMFunction_MathDoubleRemap(
     SourceMaximum=7.5
     SourceMinimum=10.1
     TargetMaximum=1
     TargetMinimum=0
     Value=&RigVMFunction_MathVectorDistance(B=@n5.Transform.Translation, A=@n6.Transform.Translation).Result
    ).Result
   )
   n7 = &RigUnit_GetTransform(Item=(Type=Bone,Name="collar_m_END"), bInitial=False)
   n8 = &RigUnit_GetTransform(Item=(Type=Bone,Name="BASE_back_hair"), bInitial=False)
   n9 = &RigUnit_GetTransform(Item=(Type=Bone,Name="BASE_collar_m"), bInitial=False)
   n10 = &RigUnit_GetTransform(Item=(Type=Bone,Name="back_hair_03"), bInitial=False)
   Maximum = &RigVMFunction_MathDoubleMax(
    A=&RigVMFunction_MathDoubleRemap(
     SourceMaximum=2.5
     SourceMinimum=5.6
     TargetMaximum=1
     TargetMinimum=0
     Value=&RigVMFunction_MathVectorDistance(B=@n7.Transform.Translation, A=@n8.Transform.Translation).Result
    ).Result
    B=&RigVMFunction_MathDoubleRemap(
     SourceMaximum=5
     SourceMinimum=8
     TargetMaximum=1
     TargetMinimum=0
     Value=&RigVMFunction_MathVectorDistance(B=@n9.Transform.Translation, A=@n10.Transform.Translation).Result
    ).Result
   )
   > &RigUnit_ModifyTransforms(
    ItemToModify=((Item=(Type=Bone,Name="BASE_collar_m"),Transform=(Rotation=(X=0,Y=0,Z=-0.4617,W=0.887),Translation=(X=0,Y=0,Z=0),Scale3D=(X=1,Y=1,Z=1))))
    WeightMinimum=0
    Weight=@Maximum.Result
   )
   Remap = &RigVMFunction_MathDoubleRemap(
    SourceMaximum=6
    SourceMinimum=7.1
    TargetMaximum=1
    TargetMinimum=0
    Value=&RigVMFunction_MathVectorDistance(
     B=&RigUnit_GetTransform(Item=(Type=Bone,Name="collar_l2_l_END"), bInitial=False).Transform.Translation
     A=&RigUnit_GetTransform(Item=(Type=Bone,Name="BASE_back_hair"), bInitial=False).Transform.Translation
    ).Result
   )
   > &RigUnit_ModifyTransforms(
    ItemToModify=((Item=(Type=Bone,Name="BASE_collar_l2_l"),Transform=(Rotation=(X=0,Y=0,Z=-0.4617,W=0.887),Translation=(X=0,Y=0,Z=0),Scale3D=(X=1,Y=1,Z=1))))
    WeightMinimum=0
    Weight=@Remap.Result
   )
   n11 = &RigUnit_GetTransform(Item=(Type=Bone,Name="BASE_back_hair"), bInitial=False)
   n12 = &RigUnit_GetTransform(Item=(Type=Bone,Name="collar_l4_l_END"), bInitial=False)
   > &RigUnit_ModifyTransforms(
    ItemToModify=((Item=(Type=Bone,Name="BASE_collar_l4_l"),Transform=(Rotation=(X=0,Y=0,Z=-0.3827,W=0.9239),Translation=(X=0,Y=0,Z=0),Scale3D=(X=1,Y=1,Z=1))))
    WeightMinimum=0
    Weight=&RigVMFunction_MathDoubleRemap(
     SourceMaximum=7.5
     SourceMinimum=10.1
     TargetMaximum=1
     TargetMinimum=0
     Value=&RigVMFunction_MathVectorDistance(A=@n11.Transform.Translation, B=@n12.Transform.Translation).Result
    ).Result
   )
   n13 = &RigUnit_GetTransform(Item=(Type=Bone,Name="head_end"), bInitial=False)
   n14 = &RigUnit_GetTransform(Item=(Type=Bone,Name="BASE_collar_m"), bInitial=False)
   > &RigUnit_ModifyTransforms(
    ItemToModify="((Item=(Type=Bone,Name=\"BASE_collar_m\"),Transform=(Rotation=(X=0,Y=0,Z=-0.1564,W=0.9877),Translation=(X=0,Y=0,Z=0),Scale3D=(X=1,Y=1,Z=1))),(Item=(Type=Bone,Name=\"BASE_collar_l2_l\"),Transform=(Rotation=(X=0,Y=0,Z=-0.1564,W=0.9877),Translation=(X=0,Y=0,Z=0), ...(truncated)"
    WeightMinimum=0
    Weight=&RigVMFunction_MathDoubleRemap(
     SourceMaximum=13
     SourceMinimum=14.6
     TargetMaximum=1
     TargetMinimum=0
     Value=&RigVMFunction_MathVectorDistance(B=@n13.Transform.Translation, A=@n14.Transform.Translation).Result
    ).Result
   )
   > n15 = &RigUnit_SphericalPoseReader(
    ActiveRegionScaleFactors=(PositiveWidth=1,NegativeWidth=0.1,PositiveHeight=1,NegativeHeight=0.5)
    ActiveRegionSize=0.3
    Debug=(bDrawDebug=False,bDraw2D=false,bDrawLocalAxes=false,DebugScale=30,DebugSegments=30,DebugThickness=0.2)
    DriverAxis=(X=1,Y=0,Z=0)
    DriverItem=(Type=Bone,Name="clavicle_l")
    FalloffRegionScaleFactors=(PositiveWidth=1,NegativeWidth=0.5,PositiveHeight=1,NegativeHeight=0.5)
    FalloffSize=0.5
    OptionalParentItem=(Type=Bone,Name="None")
    RotationOffset=(X=0,Y=40,Z=180)
   )
   AlphaInterp = &RigVMFunction_AlphaInterp(InterpSpeedDecreasing=10, InterpSpeedIncreasing=10, Scale=1, bClampResult=False, Value=@n15.OutputParam)
   > &RigUnit_ModifyTransforms(
    ItemToModify="((Item=(Type=Bone,Name=\"BASE_collar_l7_l\"),Transform=(Rotation=(X=0,Y=0,Z=0,W=1),Translation=(X=0,Y=-0.5,Z=0),Scale3D=(X=1,Y=1,Z=1))),(Item=(Type=Bone,Name=\"BASE_collar_l4_l\"),Transform=(Rotation=(X=0,Y=0,Z=0,W=1),Translation=(X=0,Y=-1,Z=0),Scale3D=(X=1,Y= ...(truncated)"
    WeightMinimum=0
    Weight=@AlphaInterp.Result
   )
   > n16 = &RigUnit_SphericalPoseReader(
    ActiveRegionScaleFactors=(PositiveWidth=1,NegativeWidth=0.1,PositiveHeight=0.5,NegativeHeight=1)
    ActiveRegionSize=0.3
    Debug=(bDrawDebug=false,bDraw2D=false,bDrawLocalAxes=false,DebugScale=30,DebugSegments=30,DebugThickness=0.2)
    DriverAxis=(X=-1,Y=0,Z=0)
    DriverItem=(Type=Bone,Name="clavicle_r")
    FalloffRegionScaleFactors=(PositiveWidth=1,NegativeWidth=0.5,PositiveHeight=1,NegativeHeight=1)
    FalloffSize=0.5
    OptionalParentItem=(Type=Bone,Name="None")
    RotationOffset=(X=0,Y=220,Z=180)
   )
   n17 = &RigVMFunction_AlphaInterp(InterpSpeedDecreasing=10, InterpSpeedIncreasing=10, Scale=1, bClampResult=False, Value=@n16.OutputParam)
   > &RigUnit_ModifyTransforms(
    ItemToModify="((Item=(Type=Bone,Name=\"BASE_collar_r4_r\"),Transform=(Rotation=(X=0,Y=0,Z=0,W=1),Translation=(X=0,Y=1,Z=0),Scale3D=(X=1,Y=1,Z=1))),(Item=(Type=Bone,Name=\"BASE_collar_r7_r\"),Transform=(Rotation=(X=0,Y=0,Z=0,W=1),Translation=(X=0,Y=0.5,Z=0),Scale3D=(X=1,Y=1, ...(truncated)"
    WeightMinimum=0
    Weight=@n17.Result
   )
   > /Script/ControlRig.RigUnit_SetRelativeTransformForItem(
    Child=(Type=Bone,Name="extra_spine_04")
    Parent=(Type=Bone,Name="extra_spine_04")
    Value=(Rotation=(X=0,Y=0,Z=0,W=1),Translation=(X=0,Y=0,Z=0),Scale3D=(X=0.3,Y=0.3,Z=0.3))
    Weight=1
    bParentInitial=False
    bPropagateToChildren=True
   ) {
    is=struct RigVMUnitNode
    template="Set Relative Transform::Execute(in Child,in Parent,in bParentInitial,in Value,in Weight,in bPropagateToChildren)"
   }
   > &RigUnit_ParentConstraint(
    AdvancedSettings=(InterpolationType=Average,RotationOrderForFilter=XZY)
    Child=(Type=Bone,Name="kneepad_l")
    Filter=(TranslationFilter=(bX=True,bY=True,bZ=True),RotationFilter=(bX=True,bY=True,bZ=True),ScaleFilter=(bX=false,bY=false,bZ=false))
    Parents=((Item=(Type=Bone,Name="thigh_l"),Weight=0.2115),(Item=(Type=Bone,Name="calf_l"),Weight=0.2437),(Item=(Type=Bone,Name="deform_knee_out_l"),Weight=0.5446),(Item=(Type=Bone,Name="deform_knee_in_l"),Weight=0.0001))
    Weight=1
   )
   > &RigUnit_ParentConstraint(
    AdvancedSettings=(InterpolationType=Average,RotationOrderForFilter=XZY)
    Child=(Type=Bone,Name="kneepad_r")
    Filter=(TranslationFilter=(bX=True,bY=True,bZ=True),RotationFilter=(bX=True,bY=True,bZ=True),ScaleFilter=(bX=False,bY=False,bZ=False))
    Parents=((Item=(Type=Bone,Name="thigh_r"),Weight=0.2115),(Item=(Type=Bone,Name="calf_r"),Weight=0.2437),(Item=(Type=Bone,Name="deform_knee_out_r"),Weight=0.5446),(Item=(Type=Bone,Name="deform_knee_in_r"),Weight=0.0001))
    Weight=1
   )
   > n18 = &RigUnit_SphericalPoseReader(
    ActiveRegionScaleFactors=(PositiveWidth=0.8,NegativeWidth=0.8,PositiveHeight=0.8,NegativeHeight=0.8)
    ActiveRegionSize=0.2
    Debug=(bDrawDebug=false,bDraw2D=false,bDrawLocalAxes=false,DebugScale=30,DebugSegments=30,DebugThickness=0.2)
    DriverAxis=(X=-1,Y=0,Z=0)
    DriverItem=(Type=Bone,Name="calf_r")
    FalloffRegionScaleFactors=(PositiveWidth=0.5,NegativeWidth=0.5,PositiveHeight=0.5,NegativeHeight=0.5)
    FalloffSize=0.5
    OptionalParentItem=(Type=Bone,Name="None")
    RotationOffset=(X=-90,Y=0,Z=50)
   )
   n19 = &RigVMFunction_AlphaInterp(InterpSpeedDecreasing=10, InterpSpeedIncreasing=10, Scale=1, bClampResult=False, Value=@n18.OutputParam)
   > &RigUnit_SetRelativeTransformForItem(
    Child=(Type=Bone,Name="kneepad_r")
    Parent=(Type=Bone,Name="kneepad_r")
    Value=(Rotation=(X=0,Y=0,Z=0.1736,W=0.9848),Translation=(X=1,Y=1,Z=0),Scale3D=(X=1,Y=1,Z=1))
    bParentInitial=False
    bPropagateToChildren=True
    Weight=@n19.Result
   ) {template="Set Relative Transform::Execute(in Child,in Parent,in bParentInitial,in Value,in Weight,in bPropagateToChildren)"}
   > n20 = &RigUnit_SphericalPoseReader(
    ActiveRegionScaleFactors=(PositiveWidth=0.8,NegativeWidth=0.8,PositiveHeight=0.8,NegativeHeight=0.8)
    ActiveRegionSize=0.2
    Debug=(bDrawDebug=false,bDraw2D=false,bDrawLocalAxes=false,DebugScale=30,DebugSegments=30,DebugThickness=0.2)
    DriverAxis=(X=1,Y=0,Z=0)
    DriverItem=(Type=Bone,Name="calf_l")
    FalloffRegionScaleFactors=(PositiveWidth=0.5,NegativeWidth=0.5,PositiveHeight=0.5,NegativeHeight=0.5)
    FalloffSize=0.5
    OptionalParentItem=(Type=Bone,Name="None")
    RotationOffset=(X=90,Y=0,Z=50)
   )
   n21 = &RigVMFunction_AlphaInterp(InterpSpeedDecreasing=10, InterpSpeedIncreasing=10, Scale=1, bClampResult=False, Value=@n20.OutputParam)
   > &RigUnit_SetRelativeTransformForItem(
    Child=(Type=Bone,Name="kneepad_l")
    Parent=(Type=Bone,Name="kneepad_l")
    Value=(Rotation=(X=0,Y=0,Z=0.1736,W=0.9848),Translation=(X=-1,Y=-1,Z=0),Scale3D=(X=1,Y=1,Z=1))
    bParentInitial=False
    bPropagateToChildren=True
    Weight=@n21.Result
   ) {template="Set Relative Transform::Execute(in Child,in Parent,in bParentInitial,in Value,in Weight,in bPropagateToChildren)"}
  ]}
]
hierarchy [
 root {loc=<0 0 0>, rot=<0 0 0 1>}
 pelvis {parent=root, loc=<0 -0.5301 98.0571>, rot=<0.0127 0.707 -0.0127 -0.707>}
 spine_01 {parent=pelvis, loc=<6.1208 -0.9328 0>, rot=<0 0 0.034 -0.9994>}
 spine_02 {parent=spine_01, loc=<6.5997 -0.2095 0>, rot=<0 0 0.0238 -0.9997>}
 spine_03 {parent=spine_02, loc=<6.4914 -0.1237 0>, rot=<0 0 0.0149 -0.9999>}
 spine_04 {parent=spine_03, loc=<6.9598 0.0794 0>, rot=<0 0 -0.0132 -0.9999>}
 spine_05 {parent=spine_04, loc=<6.9987 0 0>, rot=<0 0 -0.0015 -1>}
 neck_01 {parent=spine_05, loc=<11.3138 0.1529 0>, rot=<0 0 -0.2003 -0.9797>}
 neck_02 {parent=neck_01, loc=<4.5627 0 0>, rot=<0 0 0.0504 -0.9987>}
 head {parent=neck_02, loc=<5.3914 0 0>, rot=<0 0 0.115 -0.9934>}
 head_end {parent=head, loc=<9.2548 0 0>, rot=<0 0 0 1>}
 BASE_back_hair {parent=head, loc=<1.1193 -5.4514 0.0953>, rot=<0.365 -0.5677 -0.5653 0.4743>}
 back_hair_01 {parent=BASE_back_hair, loc=<1.5308 -0.9342 -3.1795>, rot=<0.0044 -0.3541 -0.103 -0.9295>}
 back_hair_02 {parent=back_hair_01, loc=<1.5097 -0.1764 0.0405>, rot=<0 0.1175 0 -0.9931>}
 back_hair_03 {parent=BASE_back_hair, loc=<3.99 -3.8588 1.4606>, rot=<0.9935 0.0739 -0.0639 -0.0581>}
 back_hair_04 {parent=BASE_back_hair, loc=<4.9037 1.8746 -1.9447>, rot=<0.9628 -0.0623 -0.2627 -0.0116>}
 back_hair_05 {parent=BASE_back_hair, loc=<1.9275 3.282 -1.0553>, rot=<-0.0505 -0.152 -0.3109 -0.9368>}
 NewBone {parent=BASE_back_hair, loc=<9.3879 -5.0749 9.8667>, rot=<-0.0929 -0.449 -0.1358 0.8783>}
 F_hair_01 {parent=head, loc=<7.4167 9.0306 1.0881>, rot=<-0.0916 -0.9958 0 0>}
 F_hair_02 {parent=F_hair_01, loc=<1.9623 0.1044 0.1378>, rot=<0 0 0.135 -0.9908>}
 BASE_pigtail_01 {parent=head, loc=<0.0482 5.5922 7.0788>, rot=<-0.0445 -0.9933 0.0562 -0.0909>}
 pigtail_01 {parent=BASE_pigtail_01, loc=<0 0 0>, rot=<0 0 0 1>}
 pigtail_02 {parent=pigtail_01, loc=<3.6776 0 0>, rot=<0 -0.0851 0 -0.9964>}
 pigtail_03 {parent=pigtail_02, loc=<3.3392 0 0>, rot=<-0.0385 0.0916 0.0035 -0.995>}
 pigtail_04 {parent=pigtail_03, loc=<3.0132 0 0>, rot=<-0.2012 -0.0793 -0.0164 -0.9762>}
 pigtail_05 {parent=pigtail_04, loc=<3.2096 0 0>, rot=<-0.2622 -0.1321 -0.0363 -0.9552>}
 pigtail_06 {parent=pigtail_05, loc=<3.4494 0 0>, rot=<0.1547 -0.082 0.0129 -0.9845>}
 pigtail_end {parent=pigtail_06, loc=<2.8482 0 0>, rot=<0 0 0 1>}
 L_hair_01 {parent=head, loc=<4.4873 5.9318 -6.4247>, rot=<-0.0039 -0.9999 -0.0002 -0.0145>}
 L_hair_end {parent=L_hair_01, loc=<2.3899 0 0>, rot=<0 0 0 1>}
 FACIAL_C_FacialRoot {parent=head, loc=<2.1524 -0.746 -0.0002>, rot=<0.4879 -0.5118 -0.4879 -0.5118>}
 FACIAL_C_Forehead {parent=FACIAL_C_FacialRoot, loc=<-0.0002 -3.0333 8.8872>, rot=<-0.0483 0 0 -0.9988>}
 FACIAL_L_ForeheadIn {parent=FACIAL_C_FacialRoot, loc=<1.6454 -2.9352 8.8129>, rot=<-0.0655 -0.0696 0.0039 -0.9954>}
 FACIAL_R_ForeheadIn {parent=FACIAL_C_FacialRoot, loc=<-1.6468 -2.9347 8.8119>, rot=<-0.0655 0.0696 -0.0039 -0.9954>}
 FACIAL_L_ForeheadMid {parent=FACIAL_C_FacialRoot, loc=<3.1357 -3.1749 8.3217>, rot=<-0.0758 -0.1756 -0.0215 -0.9813>}
 FACIAL_R_ForeheadMid {parent=FACIAL_C_FacialRoot, loc=<-3.1371 -3.1736 8.3199>, rot=<-0.0757 0.1756 0.0215 -0.9813>}
 FACIAL_L_ForeheadOut {parent=FACIAL_C_FacialRoot, loc=<4.5235 -3.0952 7.4483>, rot=<-0.126 -0.2552 -0.0804 -0.9553>}
 FACIAL_R_ForeheadOut {parent=FACIAL_C_FacialRoot, loc=<-4.5229 -3.0953 7.4477>, rot=<-0.126 0.2552 0.0804 -0.9553>}
 FACIAL_L_EyesackUpper {parent=FACIAL_C_FacialRoot, loc=<3.0624 -2.1612 8.8238>, rot=<0.044 -0.0839 -0.0649 -0.9934>}
 FACIAL_R_EyesackUpper {parent=FACIAL_C_FacialRoot, loc=<-3.0638 -2.1604 8.823>, rot=<0.044 0.0839 0.0649 -0.9934>}
 FACIAL_L_EyelidUpperFurrow {parent=FACIAL_C_FacialRoot, loc=<2.971 -1.8849 8.7241>, rot=<-0.0348 -0.1504 -0.0804 -0.9847>}
 FACIAL_R_EyelidUpperFurrow {parent=FACIAL_C_FacialRoot, loc=<-2.9707 -1.8818 8.7205>, rot=<-0.0348 0.1504 0.0804 -0.9847>}
 FACIAL_L_EyelidUpperA {parent=FACIAL_C_FacialRoot, loc=<3.0192 -0.735 6.6303>, rot=<-0.1526 0 0 -0.9883>}
 FACIAL_R_EyelidUpperA {parent=FACIAL_C_FacialRoot, loc=<-3.0181 -0.7366 6.6325>, rot=<-0.1526 0 0 -0.9883>}
 FACIAL_L_Eye {parent=FACIAL_C_FacialRoot, loc=<3.0188 -0.7481 6.6289>, rot=<-0.0049 0 0 -1>}
 FACIAL_R_Eye {parent=FACIAL_C_FacialRoot, loc=<-3.0212 -0.7467 6.6313>, rot=<-0.0049 0 0 -1>}
 FACIAL_L_EyelidLowerA {parent=FACIAL_C_FacialRoot, loc=<3.0192 -0.735 6.6303>, rot=<0.2794 0 0 -0.9602>}
 FACIAL_R_EyelidLowerA {parent=FACIAL_C_FacialRoot, loc=<-3.0181 -0.7366 6.6325>, rot=<0.2793 0 0 -0.9602>}
 FACIAL_L_EyeCornerInner {parent=FACIAL_C_FacialRoot, loc=<1.6033 -0.0812 7.7031>, rot=<-0.0049 0 0 -1>}
 FACIAL_R_EyeCornerInner {parent=FACIAL_C_FacialRoot, loc=<-1.5993 -0.0927 7.7107>, rot=<-0.0049 0 0 -1>}
 FACIAL_L_EyeCornerOuter {parent=FACIAL_C_FacialRoot, loc=<4.7905 -0.6273 7.2738>, rot=<-0.0045 -0.3826 -0.0018 -0.9239>}
 FACIAL_R_EyeCornerOuter {parent=FACIAL_C_FacialRoot, loc=<-4.7944 -0.6204 7.2779>, rot=<-0.0045 0.3826 0.0019 -0.9239>}
 FACIAL_L_EyesackLower {parent=FACIAL_C_FacialRoot, loc=<3.4953 1.0355 7.1877>, rot=<-0.0271 -0.1396 -0.0229 -0.9896>}
 FACIAL_R_EyesackLower {parent=FACIAL_C_FacialRoot, loc=<-3.4936 1.0356 7.1879>, rot=<-0.0271 0.1396 0.0229 -0.9896>}
 FACIAL_L_CheekInner {parent=FACIAL_C_FacialRoot, loc=<2.0017 1.6878 7.6766>, rot=<-0.1279 -0.087 -0.1975 -0.968>}
 FACIAL_R_CheekInner {parent=FACIAL_C_FacialRoot, loc=<-1.9988 1.6865 7.6759>, rot=<-0.1279 0.087 0.1975 -0.968>}
 FACIAL_L_CheekOuter {parent=FACIAL_C_FacialRoot, loc=<4.0191 1.5956 6.2709>, rot=<0.0507 -0.3774 0.0985 -0.9194>}
 FACIAL_R_CheekOuter {parent=FACIAL_C_FacialRoot, loc=<-4.0161 1.5902 6.2713>, rot=<0.0507 0.3774 -0.0985 -0.9194>}
 FACIAL_L_NasolabialBulge {parent=FACIAL_C_FacialRoot, loc=<3.1659 4.8335 6.2002>, rot=<0.0534 -0.3774 0.022 -0.9242>}
 FACIAL_R_NasolabialBulge {parent=FACIAL_C_FacialRoot, loc=<-3.1643 4.8339 6.2013>, rot=<0.0534 0.3774 -0.022 -0.9242>}
 FACIAL_L_NasolabialFurrow {parent=FACIAL_C_FacialRoot, loc=<2.7309 3.6656 8.4115>, rot=<-0.0377 -0.1241 -0.3642 -0.9222>}
 FACIAL_R_NasolabialFurrow {parent=FACIAL_C_FacialRoot, loc=<-2.7316 3.6638 8.4108>, rot=<-0.0378 0.124 0.3642 -0.9222>}
 FACIAL_L_CheekLower {parent=FACIAL_C_FacialRoot, loc=<3.5136 4.5701 5.5758>, rot=<0.0302 -0.4272 0.0557 -0.902>}
 FACIAL_R_CheekLower {parent=FACIAL_C_FacialRoot, loc=<-3.5122 4.5704 5.5767>, rot=<0.0302 0.4272 -0.0557 -0.902>}
 FACIAL_C_Nose {parent=FACIAL_C_FacialRoot, loc=<-0.0003 0.9436 8.8212>, rot=<-0.1352 0 0 -0.9908>}
 FACIAL_C_NoseLower {parent=FACIAL_C_Nose, loc=<0.0003 2.9567 -0.1972>, rot=<0.1305 0 0 -0.9914>}
 FACIAL_L_Nostril {parent=FACIAL_C_Nose, loc=<0.7794 2.4469 -0.1619>, rot=<0.0173 0.1305 0.0023 -0.9913>}
 FACIAL_R_Nostril {parent=FACIAL_C_Nose, loc=<-0.7792 2.4466 -0.1651>, rot=<0.0173 -0.1305 -0.0023 -0.9913>}
 FACIAL_C_NoseTip {parent=FACIAL_C_Nose, loc=<-0.0005 2.1463 0.4392>, rot=<0.0872 0 0 -0.9962>}
 FACIAL_C_MouthUpper {parent=FACIAL_C_FacialRoot, loc=<0.0006 5.0579 7.1552>, rot=<-0.0048 0 0 -1>}
 FACIAL_C_LipUpper {parent=FACIAL_C_MouthUpper, loc=<-0.0005 0.1223 1.8263>, rot=<-0.0872 0 0 -0.9962>}
 FACIAL_L_LipUpper {parent=FACIAL_C_MouthUpper, loc=<0.8379 0.1519 1.7404>, rot=<-0.0778 -0.1601 -0.0379 -0.9833>}
 FACIAL_R_LipUpper {parent=FACIAL_C_MouthUpper, loc=<-0.8388 0.1524 1.7389>, rot=<-0.0778 0.1601 0.0379 -0.9833>}
 FACIAL_L_LipUpperOuter {parent=FACIAL_C_MouthUpper, loc=<1.5109 0.2905 1.3138>, rot=<-0.0984 -0.2293 -0.0983 -0.9634>}
 FACIAL_R_LipUpperOuter {parent=FACIAL_C_MouthUpper, loc=<-1.5148 0.289 1.3146>, rot=<-0.0984 0.2293 0.0983 -0.9634>}
 FACIAL_L_LipCorner {parent=FACIAL_C_MouthUpper, loc=<2.0379 0.5916 0.6156>, rot=<0 -0.2078 0 -0.9782>}
 FACIAL_R_LipCorner {parent=FACIAL_C_MouthUpper, loc=<-2.0417 0.5932 0.6214>, rot=<0 0.2078 0 -0.9782>}
 FACIAL_L_Masseter {parent=FACIAL_C_FacialRoot, loc=<6.0901 1.9432 4.3453>, rot=<0.0646 0.1128 0.0084 -0.9915>}
 FACIAL_R_Masseter {parent=FACIAL_C_FacialRoot, loc=<-6.089 1.9425 4.3455>, rot=<0.0645 -0.1128 -0.0084 -0.9915>}
 FACIAL_C_UnderChin {parent=FACIAL_C_FacialRoot, loc=<0.0001 8.9394 3.9514>, rot=<0.4957 0 0 -0.8685>}
 FACIAL_L_UnderChin {parent=FACIAL_C_FacialRoot, loc=<2.7924 7.5368 3.0455>, rot=<0.3985 -0.3553 0.0806 -0.8417>}
 FACIAL_R_UnderChin {parent=FACIAL_C_FacialRoot, loc=<-2.7917 7.5344 3.0466>, rot=<0.3985 0.3553 -0.0806 -0.8417>}
 FACIAL_C_TeethUpper {parent=FACIAL_C_FacialRoot, loc=<0.0029 5.6513 8.079>, rot=<-0.0048 0 0 -1>}
 FACIAL_C_LowerLipRotation {parent=FACIAL_C_FacialRoot, loc=<-0.0016 2.3982 2.1402>, rot=<0.3375 0 0 -0.9413>}
 FACIAL_C_MouthLower {parent=FACIAL_C_LowerLipRotation, loc=<0.0021 -1.1325 5.5625>, rot=<-0.342 0 0 -0.9397>}
 FACIAL_C_LipLower {parent=FACIAL_C_MouthLower, loc=<0.0029 1.31 1.4858>, rot=<0.0872 0 0 -0.9962>}
 FACIAL_L_LipLower {parent=FACIAL_C_MouthLower, loc=<0.8242 1.2622 1.3375>, rot=<0.0927 -0.1294 0.0272 -0.9869>}
 FACIAL_R_LipLower {parent=FACIAL_C_MouthLower, loc=<-0.8186 1.2608 1.3396>, rot=<0.0927 0.1294 -0.0272 -0.9869>}
 FACIAL_L_LipLowerOuter {parent=FACIAL_C_MouthLower, loc=<1.5319 1.0492 0.9917>, rot=<0.0913 -0.2321 0.0318 -0.9679>}
 FACIAL_R_LipLowerOuter {parent=FACIAL_C_MouthLower, loc=<-1.53 1.0483 0.9948>, rot=<0.0913 0.2321 -0.0318 -0.9679>}
 FACIAL_C_TeethLower {parent=FACIAL_C_LowerLipRotation, loc=<0.004 -1.2271 6.5887>, rot=<-0.342 0 0 -0.9397>}
 FACIAL_C_Tongue1 {parent=FACIAL_C_TeethLower, loc=<-0.003 -0.1403 -4.7051>, rot=<-0.2588 0 0 -0.9659>}
 FACIAL_C_Tongue2 {parent=FACIAL_C_Tongue1, loc=<0.0004 0.4537 1.364>, rot=<0.2588 0 0 -0.9659>}
 FACIAL_C_Tongue3 {parent=FACIAL_C_Tongue2, loc=<0.0004 0.2604 1.2334>, rot=<0.1305 0 0 -0.9914>}
 FACIAL_C_Tongue4 {parent=FACIAL_C_Tongue3, loc=<0.001 -0.1082 1.0566>, rot=<0.0436 0 0 -0.999>}
 FACIAL_C_TongueUpper3 {parent=FACIAL_C_Tongue3, loc=<-0.0001 -0.3397 -0.1677>, rot=<0 0 0 1>}
 FACIAL_C_TongueLower3 {parent=FACIAL_C_Tongue3, loc=<-0.0001 0.8044 0.0135>, rot=<0 0 0 1>}
 FACIAL_L_TongueSide3 {parent=FACIAL_C_Tongue3, loc=<1.4732 0.009 0.1215>, rot=<0 0 0 1>}
 FACIAL_R_TongueSide3 {parent=FACIAL_C_Tongue3, loc=<-1.4734 0.009 0.1218>, rot=<0 0 0 1>}
 FACIAL_C_TongueUpper2 {parent=FACIAL_C_Tongue2, loc=<-0.0001 -0.5722 -0.1607>, rot=<0 0 0 1>}
 FACIAL_L_TongueSide2 {parent=FACIAL_C_Tongue2, loc=<1.7551 0.0346 0.1784>, rot=<0 0 0 1>}
 FACIAL_R_TongueSide2 {parent=FACIAL_C_Tongue2, loc=<-1.7545 0.0347 0.179>, rot=<0 0 0 1>}
 FACIAL_C_Jaw {parent=FACIAL_C_FacialRoot, loc=<-0.0016 2.3981 2.1402>, rot=<0.3375 0 0 -0.9413>}
 FACIAL_C_Jawline {parent=FACIAL_C_Jaw, loc=<0.0013 2.8931 7.1913>, rot=<0.2586 0 0 -0.966>}
 FACIAL_L_ChinSide {parent=FACIAL_C_Jaw, loc=<2.5794 0.7037 6.9654>, rot=<-0.2208 -0.3738 0.0499 -0.8995>}
 FACIAL_R_ChinSide {parent=FACIAL_C_Jaw, loc=<-2.5739 0.7032 6.9644>, rot=<-0.2208 0.3738 -0.0499 -0.8995>}
 FACIAL_C_Chin {parent=FACIAL_C_Jaw, loc=<0.0018 1.4047 7.3546>, rot=<-0.2591 0 0 -0.9659>}
 FACIAL_L_Ear {parent=FACIAL_C_FacialRoot, loc=<6.0937 0.1498 1.1188>, rot=<-0.0311 -0.0006 -0.0262 -0.9992>}
 FACIAL_R_Ear {parent=FACIAL_C_FacialRoot, loc=<-6.0875 0.1514 1.1154>, rot=<-0.0311 0.0006 0.0262 -0.9992>}
 FACIAL_L_Jawline {parent=FACIAL_C_FacialRoot, loc=<4.2098 5.7974 4.6969>, rot=<0.2046 0.158 -0.0539 -0.9645>}
 FACIAL_R_Jawline {parent=FACIAL_C_FacialRoot, loc=<-4.2119 5.7984 4.6995>, rot=<0.2046 -0.158 0.0539 -0.9645>}
 FACIAL_C_NoseBridge {parent=FACIAL_C_FacialRoot, loc=<-0.0004 -1.5003 9.591>, rot=<0.039 0 0 -0.9992>}
 FACIAL_L_EyelidUpperB {parent=FACIAL_C_FacialRoot, loc=<3.0192 -0.735 6.6303>, rot=<-0.2297 0 0 -0.9733>}
 FACIAL_L_EyelidLowerB {parent=FACIAL_C_FacialRoot, loc=<3.0192 -0.735 6.6303>, rot=<0.3127 0 0 -0.9499>}
 FACIAL_R_EyelidLowerB {parent=FACIAL_C_FacialRoot, loc=<-3.0181 -0.7366 6.6325>, rot=<0.321 0 0 -0.9471>}
 FACIAL_R_EyelidUpperB {parent=FACIAL_C_FacialRoot, loc=<-3.0181 -0.7366 6.6325>, rot=<-0.2297 0 0 -0.9733>}
 clavicle_l {parent=spine_05, loc=<5.4831 -0.1053 -3.2946>, rot=<0.0465 -0.7289 -0.009 -0.6829>}
 upperarm_l {parent=clavicle_l, loc=<9.8658 -1.2136 -0.4105>, rot=<0.0164 -0.3245 -0.0269 -0.9454>}
 lowerarm_l {parent=upperarm_l, loc=<26.465 -0.8345 0.259>, rot=<0.0036 0.0161 -0.2298 -0.9731>}
 hand_l {parent=lowerarm_l, loc=<22.7355 -0.3203 -0.6105>, rot=<-0.6784 -0.0223 0.0735 -0.7307>}
 thumb_01_l {parent=hand_l, loc=<1.3118 -0.3191 -2.459>, rot=<0.6546 -0.4042 0.0664 -0.6354>}
 thumb_02_l {parent=thumb_01_l, loc=<5.9336 -0.0448 0.0401>, rot=<-0.0037 0.0213 0.298 -0.9543>}
 thumb_03_l {parent=thumb_02_l, loc=<2.2223 0.3682 -0.0907>, rot=<0.0118 -0.013 0.0066 -0.9998>}
 index_metacarpal_l {parent=hand_l, loc=<1.6535 0.0183 -2.96>, rot=<0.0158 -0.0124 -0.0173 -0.9996>}
 index_01_l {parent=index_metacarpal_l, loc=<8.3008 0.0357 -0.0524>, rot=<-0.0044 0.0485 0.0924 -0.9945>}
 index_02_l {parent=index_01_l, loc=<3.3587 -0.3004 0.0452>, rot=<-0.002 0.0027 0.3225 -0.9466>}
 index_03_l {parent=index_02_l, loc=<2.5578 -0.123 -0.0252>, rot=<-0.002 0.0067 0.2558 -0.9667>}
 middle_metacarpal_l {parent=hand_l, loc=<2.1856 -0.1725 -0.5984>, rot=<0.006 0.0168 -0.0329 -0.9993>}
 middle_01_l {parent=middle_metacarpal_l, loc=<7.3275 -0.2447 0.0129>, rot=<-0.0208 0.0042 0.0858 -0.9961>}
 middle_02_l {parent=middle_01_l, loc=<4.0123 0.1401 0.1175>, rot=<-0.0218 0.0258 0.3512 -0.9357>}
 middle_03_l {parent=middle_02_l, loc=<2.6077 0.0534 0.1373>, rot=<0.0082 0.0188 0.1708 -0.9851>}
 ring_metacarpal_l {parent=hand_l, loc=<2.2074 -0.3625 1.3638>, rot=<-0.0284 0.0722 -0.0458 -0.9959>}
 ring_01_l {parent=ring_metacarpal_l, loc=<7.0295 -0.6197 -0.1446>, rot=<-0.0489 -0.0252 0.1892 -0.9804>}
 ring_02_l {parent=ring_01_l, loc=<3.6174 0.4647 -0.1027>, rot=<0 0 0.2806 -0.9598>}
 ring_03_l {parent=ring_02_l, loc=<2.4152 0.114 0.0199>, rot=<0 0 0.1189 -0.9929>}
 pinky_metacarpal_l {parent=hand_l, loc=<1.7616 -0.9753 3.3814>, rot=<-0.0397 0.0814 -0.069 -0.9935>}
 pinky_01_l {parent=pinky_metacarpal_l, loc=<7.1709 -0.4831 -0.1094>, rot=<-0.0654 -0.0035 0.2242 -0.9723>}
 pinky_02_l {parent=pinky_01_l, loc=<2.8274 0.0893 -0.1043>, rot=<0.0037 0.0011 0.2854 -0.9584>}
 pinky_03_l {parent=pinky_02_l, loc=<2.2046 0.0307 -0.0078>, rot=<-0.0002 0.0098 0.1445 -0.9895>}
 deform_wrist_dn_base_l {parent=hand_l, loc=<0 -1.5326 0>, rot=<0 0 0 -1>}
 deform_wrist_dn_l {parent=deform_wrist_dn_base_l, loc=<0 0 0>, rot=<0 0 0 1>}
 deform_wrist_up_base_l {parent=hand_l, loc=<0 2.3009 0>, rot=<0 0 0 -1>}
 deform_wrist_up_l {parent=deform_wrist_up_base_l, loc=<0 0 0>, rot=<0 0 0 1>}
 hand_attach_l {parent=hand_l, loc=<8.5 -3 0>, rot=<0.7071 0.7071 0 0>}
 weapon_l {parent=hand_l, loc=<-0.1974 -3.4111 0.3811>, rot=<0 0 0 1>}
 lowerarm_twist_02_l {parent=lowerarm_l, loc=<10.6707 0 0>, rot=<0 0 0 1>}
 lowerarm_twist_01_l {parent=lowerarm_l, loc=<18.316 0 0>, rot=<0 0 0 1>}
 deform_elbow_in_base_l {parent=lowerarm_l, loc=<0.1834 3.6062 -0.2572>, rot=<0 0 0 -1>}
 deform_elbow_in_l {parent=deform_elbow_in_base_l, loc=<0 0 0>, rot=<0 0 0 1>}
 deform_elbow_out_base_l {parent=lowerarm_l, loc=<0 -4 0>, rot=<0 0 0 1>}
 deform_elbow_out_l {parent=deform_elbow_out_base_l, loc=<0 0 0>, rot=<0 0 0 1>}
 upperarm_twist_01_l {parent=upperarm_l, loc=<7.2135 -0.1134 0>, rot=<0 0 0 1>}
 deform_upperarm_bk_base_l {parent=upperarm_twist_01_l, loc=<-6.4053 -4.3746 0.314>, rot=<0 0 0 -1>}
 deform_upperarm_bk_l {parent=deform_upperarm_bk_base_l, loc=<0 0 0>, rot=<0 0 0 1>}
 deform_upperarm_fr_base_l {parent=upperarm_twist_01_l, loc=<-5.6312 5.092 0.3354>, rot=<0 0 0 -1>}
 deform_upperarm_fr_l {parent=deform_upperarm_fr_base_l, loc=<0 0 0>, rot=<0 0 0 1>}
 deform_upperarm_up_base_l {parent=upperarm_twist_01_l, loc=<-8.891 0.4825 3.9517>, rot=<0 0 0 -1>}
 deform_upperarm_up_l {parent=deform_upperarm_up_base_l, loc=<0 0 0>, rot=<0 0 0 1>}
 upperarm_twist_02_l {parent=upperarm_l, loc=<19.7998 -0.1134 0>, rot=<0 0 0 1>}
 BASE_binoculars {parent=upperarm_l, loc=<5.2013 -0.5502 4.3269>, rot=<0 0 0 1>}
 binoculars {parent=BASE_binoculars, loc=<0 0 0>, rot=<0 0 0 1>}
 clavicle_r {parent=spine_05, loc=<5.4831 -0.1053 3.2946>, rot=<0.7289 0.0465 -0.6829 0.009>}
 upperarm_r {parent=clavicle_r, loc=<-9.8658 1.2136 0.4105>, rot=<0.0164 -0.3245 -0.0269 -0.9454>}
 lowerarm_r {parent=upperarm_r, loc=<-26.465 0.8345 -0.259>, rot=<0.0036 0.0161 -0.2298 -0.9731>}
 hand_r {parent=lowerarm_r, loc=<-22.7355 0.3203 0.6106>, rot=<-0.6784 -0.0223 0.0735 -0.7307>}
 thumb_01_r {parent=hand_r, loc=<-1.3118 0.3191 2.459>, rot=<0.6546 -0.4042 0.0664 -0.6354>}
 thumb_02_r {parent=thumb_01_r, loc=<-5.9336 0.0448 -0.0401>, rot=<-0.0037 0.0213 0.298 -0.9543>}
 thumb_03_r {parent=thumb_02_r, loc=<-2.2223 -0.3682 0.0907>, rot=<0.0118 -0.013 0.0066 -0.9998>}
 index_metacarpal_r {parent=hand_r, loc=<-1.6535 -0.0183 2.96>, rot=<0.0158 -0.0124 -0.0173 -0.9996>}
 index_01_r {parent=index_metacarpal_r, loc=<-8.3008 -0.0357 0.0524>, rot=<-0.0044 0.0485 0.0924 -0.9945>}
 index_02_r {parent=index_01_r, loc=<-3.3587 0.3004 -0.0452>, rot=<-0.002 0.0027 0.3225 -0.9466>}
 index_03_r {parent=index_02_r, loc=<-2.5578 0.123 0.0252>, rot=<-0.002 0.0067 0.2558 -0.9667>}
 middle_metacarpal_r {parent=hand_r, loc=<-2.1856 0.1725 0.5984>, rot=<0.006 0.0168 -0.0329 -0.9993>}
 middle_01_r {parent=middle_metacarpal_r, loc=<-7.3275 0.2447 -0.0129>, rot=<-0.0208 0.0042 0.0858 -0.9961>}
 middle_02_r {parent=middle_01_r, loc=<-4.0123 -0.1401 -0.1175>, rot=<-0.0218 0.0258 0.3512 -0.9357>}
 middle_03_r {parent=middle_02_r, loc=<-2.6077 -0.0534 -0.1373>, rot=<0.0082 0.0188 0.1708 -0.9851>}
 ring_metacarpal_r {parent=hand_r, loc=<-2.2074 0.3625 -1.3638>, rot=<-0.0284 0.0722 -0.0458 -0.9959>}
 ring_01_r {parent=ring_metacarpal_r, loc=<-7.0295 0.6197 0.1446>, rot=<-0.0489 -0.0252 0.1892 -0.9804>}
 ring_02_r {parent=ring_01_r, loc=<-3.6174 -0.4647 0.1027>, rot=<0 0 0.2806 -0.9598>}
 ring_03_r {parent=ring_02_r, loc=<-2.4152 -0.114 -0.0199>, rot=<0 0 0.1189 -0.9929>}
 pinky_metacarpal_r {parent=hand_r, loc=<-1.7616 0.9753 -3.3814>, rot=<-0.0397 0.0814 -0.069 -0.9935>}
 pinky_01_r {parent=pinky_metacarpal_r, loc=<-7.1709 0.4831 0.1094>, rot=<-0.0654 -0.0035 0.2242 -0.9723>}
 pinky_02_r {parent=pinky_01_r, loc=<-2.8274 -0.0893 0.1043>, rot=<0.0037 0.0011 0.2854 -0.9584>}
 pinky_03_r {parent=pinky_02_r, loc=<-2.2046 -0.0307 0.0078>, rot=<-0.0002 0.0098 0.1445 -0.9895>}
 deform_wrist_dn_base_r {parent=hand_r, loc=<-0.0001 1.5328 0>, rot=<0 0 0 -1>}
 deform_wrist_dn_r {parent=deform_wrist_dn_base_r, loc=<0 0 0>, rot=<0 0 0 1>}
 deform_wrist_up_base_r {parent=hand_r, loc=<-0.0002 -2.3007 -0.0001>, rot=<0 0 0 -1>}
 deform_wrist_up_r {parent=deform_wrist_up_base_r, loc=<0 0 0>, rot=<0 0 0 1>}
 weapon_r {parent=hand_r, loc=<0.1974 3.4111 -0.3811>, rot=<0 0 0 1>}
 hand_attach_r {parent=hand_r, loc=<-8.5001 2.9999 0>, rot=<0 0 -0.7071 -0.7071>}
 lowerarm_twist_02_r {parent=lowerarm_r, loc=<-10.6707 0 0>, rot=<0 0 0 -1>}
 lowerarm_twist_01_r {parent=lowerarm_r, loc=<-18.316 0 0>, rot=<0 0 0 -1>}
 deform_elbow_in_base_r {parent=lowerarm_r, loc=<-0.1834 -3.6062 0.2572>, rot=<0 0 0 -1>}
 deform_elbow_in_r {parent=deform_elbow_in_base_r, loc=<0 0 0>, rot=<0 0 0 1>}
 deform_elbow_out_base_r {parent=lowerarm_r, loc=<0 4 -0.0001>, rot=<0 0 0 -1>}
 deform_elbow_out_r {parent=deform_elbow_out_base_r, loc=<0 0 0>, rot=<0 0 0 1>}
 upperarm_twist_01_r {parent=upperarm_r, loc=<-7.2135 0.1134 0>, rot=<0 0 0 -1>}
 deform_upperarm_bk_base_r {parent=upperarm_twist_01_r, loc=<6.4053 4.3746 -0.314>, rot=<0 0 0 -1>}
 deform_upperarm_bk_r {parent=deform_upperarm_bk_base_r, loc=<0 0 0>, rot=<0 0 0 1>}
 deform_upperarm_fr_base_r {parent=upperarm_twist_01_r, loc=<5.6312 -5.092 -0.3354>, rot=<0 0 0 -1>}
 deform_upperarm_fr_r {parent=deform_upperarm_fr_base_r, loc=<0 0 0>, rot=<0 0 0 1>}
 deform_upperarm_up_base_r {parent=upperarm_twist_01_r, loc=<8.891 -0.4825 -3.9517>, rot=<0 0 0 -1>}
 deform_upperarm_up_r {parent=deform_upperarm_up_base_r, loc=<0 0 0>, rot=<0 0 0 1>}
 upperarm_twist_02_r {parent=upperarm_r, loc=<-19.7998 0.1134 0>, rot=<0 0 0 -1>}
 attach_backpack {parent=spine_05, loc=<-0.0953 -13.4502 -0.0046>, rot=<-0.0093 -0.707 0.0092 -0.7071>}
 Backpack_BR {parent=attach_backpack, type=Null, loc=<0 4.0752 -1.1343>, rot=<0 0 0 1>}
 attach_cape {parent=spine_05, loc=<9.0354 -11.7004 -0.0076>, rot=<0.0616 -0.7044 -0.0617 -0.7044>}
 cape {parent=attach_cape, type=Null, loc=<0 0 0>, rot=<0 0 0 1>}
 BASE_collar_m {parent=spine_05, loc=<9.991 -6.7419 0>, rot=<0.0002 -0.002 0.0198 -0.9998>}
 collar_m {parent=BASE_collar_m, loc=<0 0 0>, rot=<0 0 0 1>}
 collar_m_END {parent=collar_m, loc=<7.1702 0 0>, rot=<0 0 0 1>}
 BASE_collar_l4_l {parent=spine_05, loc=<10.0123 -0.3596 -7.9517>, rot=<0.1478 -0.6144 0.7745 -0.0279>}
 collar_l4_l {parent=BASE_collar_l4_l, loc=<0 0 0>, rot=<0 0 0 1>}
 collar_l4_l_END {parent=collar_l4_l, loc=<-5.0371 0 0>, rot=<0 0 0 1>}
 BASE_collar_l2_l {parent=spine_05, loc=<10.1085 -4.6062 -5.8197>, rot=<0.04 -0.3145 0.9482 0.0218>}
 collar_l2_l {parent=BASE_collar_l2_l, loc=<0 0 0>, rot=<0 0 0 1>}
 collar_l2_l_END {parent=collar_l2_l, loc=<-6.5456 0 0>, rot=<0 0 0 1>}
 BASE_collar_r4_r {parent=spine_05, loc=<10.0123 -0.3596 7.9517>, rot=<0.6143 0.1504 -0.03 -0.774>}
 collar_r4_r {parent=BASE_collar_r4_r, loc=<0 0 0>, rot=<0 0 0 1>}
 collar_r4_r END {parent=collar_r4_r, loc=<5.1643 0 0>, rot=<0 0 0 1>}
 BASE_collar_r2_r {parent=spine_05, loc=<10.1085 -4.6062 5.8197>, rot=<0.3144 0.0365 0.023 -0.9483>}
 collar_r2_r {parent=BASE_collar_r2_r, loc=<0 0 0>, rot=<0 0 0 1>}
 collar_r2_r_END {parent=collar_r2_r, loc=<6.4758 0 0>, rot=<0 0 0 1>}
 BASE_collar_l7_l {parent=spine_05, loc=<7.1595 6.1923 -4.8084>, rot=<0.1782 -0.919 0.3393 -0.0929>}
 collar_l7_l {parent=BASE_collar_l7_l, loc=<0 0 0>, rot=<0 0 0 1>}
 collar_l7_l_END {parent=collar_l7_l, loc=<0 0 0>, rot=<0 0 0 -1>}
 BASE_collar_r7_r {parent=spine_05, loc=<7.1598 6.1924 4.8084>, rot=<0.938 0.1766 -0.0892 -0.2845>}
 collar_r7_r {parent=BASE_collar_r7_r, loc=<0 0 0>, rot=<0 0 0 1>}
 collar_r7_r_END {parent=collar_r7_r, loc=<0 0 0>, rot=<0 0 0 -1>}
 deform_pec_aim_base_l {parent=spine_04, loc=<3.2456 9.3309 -1.2237>, rot=<0.0603 -0.4033 0.3318 -0.8507>}
 deform_pec_aim_l {parent=deform_pec_aim_base_l, loc=<0 0 0>, rot=<0 0 0 1>}
 deform_pec_base_l {parent=deform_pec_aim_l, loc=<9.3665 0 0>, rot=<0 0 0 1>}
 deform_pec_l {parent=deform_pec_base_l, loc=<0 0 0>, rot=<0 0 0 1>}
 deform_pec_aim_base_r {parent=spine_04, loc=<3.2452 9.3309 1.2237>, rot=<-0.4033 -0.0603 0.8507 0.3318>}
 deform_pec_aim_r {parent=deform_pec_aim_base_r, loc=<0 0 0>, rot=<0 0 0 1>}
 deform_pec_base_r {parent=deform_pec_aim_r, loc=<-9.3667 -0.0003 -0.0004>, rot=<0 0 0 1>}
 deform_pec_r {parent=deform_pec_base_r, loc=<0 0 0>, rot=<0 0 0 1>}
 extra_spine_04 {parent=spine_04, loc=<0 0 0>, rot=<0 0 0 1>}
 BASE_L_back_jacket {parent=spine_03, loc=<7.6282 -7.8882 -5.2446>, rot=<0.0165 -0.9998 -0.0006 -0.0067>}
 L_back_jacket_01 {parent=BASE_L_back_jacket, loc=<0 0 0>, rot=<0 0 0 1>}
 BASE_R_back_jacket {parent=spine_03, loc=<7.4689 -7.1912 5.2692>, rot=<0.0165 -0.9998 -0.0006 -0.0067>}
 R_back_jacket_01 {parent=BASE_R_back_jacket, loc=<0 0 0>, rot=<0 0 0 1>}
 BASE_R_side_jacket {parent=spine_03, loc=<4.7486 0.2504 11.602>, rot=<0.0038 -0.0701 0.996 -0.0546>}
 R_side_jacket_01 {parent=BASE_R_side_jacket, loc=<0 0 0>, rot=<0 0 0 1>}
 BASE_L_side_jacket {parent=spine_03, loc=<4.7486 0.2504 -11.602>, rot=<0.0547 -0.9977 0.0398 -0.0022>}
 L_side_jacket_01 {parent=BASE_L_side_jacket, loc=<0 0 0>, rot=<0 0 0 1>}
 BASE_L_front_jacket {parent=spine_03, loc=<6.939 10.8518 -9.204>, rot=<0.0426 -0.9373 -0.3427 -0.0462>}
 L_front_jacket_01_dyn {parent=BASE_L_front_jacket, loc=<0 0 0>, rot=<0 0 0 1>}
 L_front_jacket_02_dyn {parent=L_front_jacket_01_dyn, loc=<5.6485 0 0>, rot=<0 0 0 1>}
 BASE_BASE_zipper {parent=L_front_jacket_02_dyn, loc=<5.5277 3.5095 -0.5542>, rot=<0.3186 -0.0427 -0.0069 -0.9469>}
 BASE_zipper {parent=BASE_BASE_zipper, loc=<0 0 0>, rot=<0 0 0 1>}
 zipper {parent=BASE_zipper, loc=<0 0 0>, rot=<0 0 0 1>}
 BASE_R_front_jacket {parent=spine_03, loc=<6.939 10.8518 9.204>, rot=<0.0291 0.3984 0.9159 -0.0389>}
 R_front_jacket_01 {parent=BASE_R_front_jacket, loc=<0 0 0>, rot=<0 0 0 1>}
 R_front_jacket_02 {parent=R_front_jacket_01, loc=<5.7429 0 0>, rot=<0 0 0 1>}
 thigh_l {parent=pelvis, loc=<-1.0855 -0.2109 -8.7053>, rot=<0.0009 0.0355 0.0155 -0.9993>}
 calf_l {parent=thigh_l, loc=<-42.5284 -0.0019 -0.7297>, rot=<0.0006 -0.0161 -0.0391 -0.9991>}
 foot_l {parent=calf_l, loc=<-46.7401 -2.5014 -0.592>, rot=<0.0427 -0.0238 0.0353 -0.9982>}
 ball_l {parent=foot_l, loc=<-3.5658 10.8066 -0.1544>, rot=<0.0001 -0.0001 -0.7063 -0.7079>}
 calf_twist_02_l {parent=calf_l, loc=<-19.1546 -1.7943 -0.2638>, rot=<0.0237 0.0095 -0.0035 -0.9997>}
 calf_twist_01_l {parent=calf_l, loc=<-35.8483 -1.9197 -0.5794>, rot=<0.0237 0.0095 -0.0035 -0.9997>}
 deform_knee_in_base_l {parent=calf_l, loc=<0.8031 -5.0231 0.2456>, rot=<0 0 0 -1>}
 deform_knee_in_l {parent=deform_knee_in_base_l, loc=<0 0 0>, rot=<0 0 0 1>}
 deform_knee_out_base_l {parent=calf_l, loc=<0 4 0>, rot=<0 0 0 -1>}
 deform_knee_out_l {parent=deform_knee_out_base_l, loc=<0 0 0>, rot=<0 0 0 1>}
 kneepad_l {parent=calf_l, loc=<0 3.9 0>, rot=<0 0 0 -1>}
 thigh_twist_01_l {parent=thigh_l, loc=<-27.952 0 0>, rot=<0.0013 0 0 -1>}
 deform_groin_base_l {parent=thigh_l, loc=<0.4801 5.6776 -0.2486>, rot=<0 0 0 -1>}
 deform_groin_l {parent=deform_groin_base_l, loc=<0 0 0>, rot=<0 0 0 1>}
 thigh_r {parent=pelvis, loc=<-1.0855 -0.2109 8.7053>, rot=<0.0355 -0.0009 0.9993 0.0155>}
 calf_r {parent=thigh_r, loc=<42.5284 0.0019 0.7297>, rot=<0.0006 -0.0161 -0.0391 -0.9991>}
 foot_r {parent=calf_r, loc=<46.7401 2.5014 0.592>, rot=<0.0427 -0.0238 0.0353 -0.9982>}
 ball_r {parent=foot_r, loc=<3.5658 -10.8066 0.1544>, rot=<0.0001 -0.0001 -0.7063 -0.7079>}
 calf_twist_02_r {parent=calf_r, loc=<19.1546 1.7943 0.2638>, rot=<0.0237 0.0095 -0.0035 -0.9997>}
 calf_twist_01_r {parent=calf_r, loc=<35.8483 1.9197 0.5794>, rot=<0.0237 0.0095 -0.0035 -0.9997>}
 deform_knee_in_base_r {parent=calf_r, loc=<-0.8031 5.0231 -0.2457>, rot=<0 0 0 -1>}
 deform_knee_in_r {parent=deform_knee_in_base_r, loc=<0 0 0>, rot=<0 0 0 1>}
 deform_knee_out_base_r {parent=calf_r, loc=<0 -4 0>, rot=<0 0 0 -1>}
 deform_knee_out_r {parent=deform_knee_out_base_r, loc=<0 0 0>, rot=<0 0 0 1>}
 kneepad_r {parent=calf_r, loc=<0 -3.9 0>, rot=<0 0 0 -1>}
 thigh_twist_01_r {parent=thigh_r, loc=<27.952 0 0>, rot=<0.0013 0 0 -1>}
 deform_groin_base_r {parent=thigh_r, loc=<-0.4801 -5.6776 0.2486>, rot=<0 0 0 -1>}
 deform_groin_r {parent=deform_groin_base_r, loc=<0 0 0>, rot=<0 0 0 1>}
 deform_glute_base_l {parent=pelvis, loc=<-7.3466 -9.0213 -9.1245>, rot=<0.0172 0.031 -0.4477 -0.8935>}
 deform_glute_l {parent=deform_glute_base_l, loc=<0 0 0>, rot=<0 0 0 1>}
 deform_glute_base_r {parent=pelvis, loc=<-7.3464 -9.0213 9.1245>, rot=<0.031 -0.0172 0.8935 -0.4477>}
 deform_glute_r {parent=deform_glute_base_r, loc=<0 0 0>, rot=<0 0 0 1>}
 BASE_belt_03 {parent=pelvis, loc=<5.3466 -8.8936 -8.2231>, rot=<-0.0155 -0.9993 0.0009 -0.0355>}
 belt_03 {parent=BASE_belt_03, loc=<0 0 0>, rot=<0 0 0 1>}
 BASE_belt_01 {parent=pelvis, loc=<3.27 6.2236 -8.4146>, rot=<-0.0355 0.0009 0.9993 0.0155>}
 belt_01 {parent=BASE_belt_01, loc=<0 0 0>, rot=<0 0 0 1>}
 BASE_belt_04 {parent=pelvis, loc=<6.0785 8.5165 -0.0003>, rot=<-0.0355 0.0009 0.9993 0.0155>}
 belt_04 {parent=BASE_belt_04, loc=<0 0 0>, rot=<0 0 0 1>}
 BASE_pot_1 {parent=pelvis, loc=<6.0347 8.4813 -1.4676>, rot=<-0.0127 -0.707 0.0127 -0.707>}
 pot_1 {parent=BASE_pot_1, loc=<0 0 0>, rot=<0 0 0 1>}
 BASE_pot_3 {parent=pelvis, loc=<3.3321 6.4268 -9.4531>, rot=<-0.0127 -0.707 0.0127 -0.707>}
 pot_3 {parent=BASE_pot_3, loc=<0 0 0>, rot=<0 0 0 1>}
 BASE_pot_2 {parent=pelvis, loc=<3.3698 7.4581 -5.5543>, rot=<-0.0127 -0.707 0.0127 -0.707>}
 pot_2 {parent=BASE_pot_2, loc=<0 0 0>, rot=<0 0 0 1>}
 BASE_bag {parent=pelvis, loc=<6.0047 -0.9365 -13.1148>, rot=<0.0009 0.0355 0.0155 -0.9993>}
 bag {parent=BASE_bag, loc=<0 0 0>, rot=<0 0 0 1>}
 BASE_belt_02 {parent=bag, loc=<0 0 0>, rot=<0.0587 -0.7047 -0.7047 -0.0587>}
 belt_02 {parent=BASE_belt_02, loc=<0 0 0>, rot=<0 0 0 1>}
 rotate_1 {parent=bag, loc=<0 1.3477 -1.2304>, rot=<0 0 0 1>}
 BASE_knif1 {parent=rotate_1, loc=<-8.7736 3.8701 -0.2102>, rot=<-0.2279 -0.6575 -0.1652 -0.6989>}
 knif1 {parent=BASE_knif1, loc=<0 0 0>, rot=<0 0 0 1>}
 rotate_2 {parent=bag, loc=<-0.0134 -1.3468 -1.2314>, rot=<-0.0049 -1 -0.0004 0>}
 BASE_knif2 {parent=rotate_2, loc=<8.9211 -2.8481 0.9478>, rot=<-0.1561 -0.6781 -0.0894 -0.7126>}
 knif2 {parent=BASE_knif2, loc=<0 0 0>, rot=<0 0 0 1>}
 attach {parent=root, loc=<0 0 0>, rot=<0 0 0 1>}
 ik_hand_root {parent=root, loc=<0 0 0>, rot=<0 0 0 1>}
 ik_hand_gun {parent=ik_hand_root, loc=<-48.1743 3.3545 104.4381>, rot=<0.6276 0.1309 0.4183 -0.6434>}
 ik_hand_r {parent=ik_hand_gun, loc=<0 0 0>, rot=<0 0 0 1>}
 ik_hand_l {parent=ik_hand_gun, loc=<59.3271 67.6997 34.3524>, rot=<-0.6158 -0.7027 -0.3565 0>}
 ik_foot_root {parent=root, loc=<0 0 0>, rot=<0 0 0 1>}
 ik_foot_l {parent=ik_foot_root, loc=<14.8262 -7.3545 8.296>, rot=<0.0357 0.7045 0.0284 -0.7082>}
 ik_foot_r {parent=ik_foot_root, loc=<-14.8262 -7.3545 8.296>, rot=<0.7082 0.0284 -0.7045 0.0357>}
 root_fp {parent=root, loc=<0 0 0>, rot=<0 0 0 1>}
 attach_fp {parent=root_fp, loc=<0 0 0>, rot=<0 0 0 1>}
 camera_root {parent=root_fp, loc=<0 0 0>, rot=<0 0 0 1>}
 camera {parent=camera_root, loc=<0 0 0>, rot=<0 0 0 1>}
 attach_emote_01 {parent=root, loc=<0 0 0>, rot=<0 0 0 1>}
 attach_emote_02 {parent=root, loc=<0 0 0>, rot=<0 0 0 1>}
 attach_emote_03 {parent=root, loc=<0 0 0>, rot=<0 0 0 1>}
 warp_point_01 {parent=root, loc=<0 0 0>, rot=<0 0 0 1>}
 warp_point_02 {parent=root, loc=<0 0 0>, rot=<0 0 0 1>}
 warp_point_03 {parent=root, loc=<0 0 0>, rot=<0 0 0 1>}
]
