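# Physics asset for the bag on the female medium VanceGuard player cosmetic body, defining its collision shapes and joint limits.
# The bag box hangs from a tiny BASE_bag anchor box through a constraint with swing and twist limits.
# Two child boxes, rotate_1 and rotate_2, attach to the bag with limited twist and zero-degree swing limits.
# Unconstrained pelvis and left thigh capsules provide body shapes alongside the bag chain.
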PhysicsAsset /BRCosmetics/Characters/Player/Female/Medium/Bodies/F_MED_VanceGuard/Meshes/F_MED_VanceGuard_Physics_bag
bodies {
 BASE_bag [{shapeType=Box, extentX=1.01, extentY=1.01, extentZ=1.01, shapeName=BASE_bag_box, rotation=<rot 0 -90 0>}]
 bag [{
   shapeType=Box
   extentX=13.1727
   extentY=9.7197
   extentZ=2.1346
   shapeName=sumka_box
   center=<-12.443 -0.0355 -3.5403>
   rotation=<rot 0 -90 -7.8853>
  }]
 pelvis [{shapeType=Capsule, length=9.6958, radius=9.9439, center=<-2.5532 -1.8293 0>}]
 rotate_1 [{
   shapeType=Box
   extentX=5.8168
   extentY=2.6287
   extentZ=14.4881
   shapeName=rotate_1_box
   center=<-10.8112 2.5488 -0.8382>
   rotation=<rot 27.3388 92.8581 -80.9255>
  }]
 rotate_2 [{
   shapeType=Box
   extentX=7.2328
   extentY=3.0427
   extentZ=14.29
   shapeName=rotate_2_box
   center=<10.6642 -2.9051 0.7669>
   rotation=<rot 21.0866 86.9798 -85.501>
  }]
 thigh_l [{shapeType=Capsule, length=24.7611, radius=8.0497, center=<-14.1548 -0.5285 0.4487>, rotation=<rot 0.0001 -88.0954 -92.3192>}]
}
constraints [
 {bone1Name=bag, bone2Name=BASE_bag, swing1LimitDegrees=10, swing2LimitDegrees=45, twistLimitDegrees=5}
 {
  bone1Name=rotate_1
  bone2Name=bag
  swing1LimitDegrees=0
  swing1Motion=Locked
  swing2LimitDegrees=0
  swing2Motion=Locked
  twistLimitDegrees=20
 }
 {bone1Name=rotate_2, bone2Name=bag, swing1LimitDegrees=0, swing2LimitDegrees=0, twistLimitDegrees=10}
]
