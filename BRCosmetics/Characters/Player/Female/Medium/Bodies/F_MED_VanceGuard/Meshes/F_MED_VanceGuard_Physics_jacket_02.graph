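# Physics asset for the second jacket setup of the female medium VanceGuard body, defining bodies and joints for moving jacket panels.
# Tiny BASE capsules anchor six chains at left and right front, side and back, with two links in each front chain and one elsewhere.
# Constraints give each link narrow swing and twist limits.
# Two unconstrained spine capsules add torso volume, likely as colliders for the panels.
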PhysicsAsset /BRCosmetics/Characters/Player/Female/Medium/Bodies/F_MED_VanceGuard/Meshes/F_MED_VanceGuard_Physics_jacket_02
bodies {
 BASE_L_back_jacket [{shapeType=Capsule, length=0.505, radius=0.505, shapeName=BASE_L_back_jacket_capsule, rotation=<rot 0 -90 0>}]
 BASE_L_front_jacket [{shapeType=Capsule, length=0.505, radius=0.505, shapeName=BASE_L_front_jacket_capsule, rotation=<rot 0 -90 0>}]
 BASE_L_side_jacket [{shapeType=Capsule, length=0.505, radius=0.505, shapeName=BASE_L_side_jacket_capsule, rotation=<rot 0 -90 0>}]
 BASE_R_back_jacket [{shapeType=Capsule, length=0.505, radius=0.505, shapeName=BASE_R_back_jacket_capsule, rotation=<rot 0 -90 0>}]
 BASE_R_front_jacket [{shapeType=Capsule, length=0.505, radius=0.505, shapeName=BASE_R_front_jacket_capsule, rotation=<rot 0 -90 0>}]
 BASE_R_side_jacket [{shapeType=Capsule, length=0.505, radius=0.505, shapeName=BASE_R_side_jacket_capsule, rotation=<rot 0 -90 0>}]
 L_back_jacket_01 [{
   shapeType=Capsule
   length=4.9957
   radius=4.2861
   shapeName=L_back_jacket_03_capsule
   center=<4.4312 -2.3251 0.2693>
   rotation=<rot 9.938 -90 4.6539>
  }]
 L_front_jacket_01_dyn [{
   shapeType=Capsule
   length=3.2518
   radius=3.6053
   shapeName=L_front_jacket_01_capsule
   center=<2.5279 0.3236 1.6233>
   rotation=<rot -0.2475 -174.5133 -89.8771>
  }]
 L_front_jacket_02_dyn [{
   shapeType=Capsule
   length=2.1835
   radius=3.6267
   shapeName=L_front_jacket_02_capsule
   center=<2.2483 0.0177 2.377>
   rotation=<rot -4.219 6.3319 -90.8878>
  }]
 L_side_jacket_01 [{
   shapeType=Capsule
   length=6.7592
   radius=4.1567
   shapeName=L_side_jacket_02_capsule
   center=<2.9124 0.0231 2.3795>
   rotation=<rot -0.0323 170.3415 -89.5529>
  }]
 R_back_jacket_01 [{
   shapeType=Capsule
   length=6.1056
   radius=3.2241
   shapeName=R_back_jacket_03_capsule
   center=<3.86 -2.1135 -0.4509>
   rotation=<rot -12.634 -90 -1.2398>
  }]
 R_front_jacket_01 [{
   shapeType=Capsule
   length=0.8
   radius=4.2891
   shapeName=R_front_jacket_01_capsule
   center=<3.4391 -0.1234 2.4654>
   rotation=<rot 4.9572 -0.4512 -89.6627>
  }]
 R_front_jacket_02 [{
   shapeType=Capsule
   length=4.2698
   radius=3.4826
   shapeName=R_front_jacket_02_capsule
   center=<3.0386 1.2485 1.9383>
   rotation=<rot 0 -2.1579 -87.9535>
  }]
 R_side_jacket_01 [{
   shapeType=Capsule
   length=6.9228
   radius=4.3578
   shapeName=R_side_jacket_02_capsule
   center=<4.1457 -0.7176 2.3139>
   rotation=<rot 0.5566 -175.9176 -87.4302>
  }]
 spine_03 [{shapeType=Capsule, length=2.0493, radius=9.0507, shapeName=spine_03_clone_01_capsule, center=<0 2.3514 0>, rotation=<rot 0 -90 0>}]
 spine_04 [{shapeType=Capsule, length=3.5, radius=9.6, shapeName=spine_04_clone_01_capsule, center=<0 2.1812 0>, rotation=<rot 0 -90 0>}]
}
constraints [
 {bone1Name=L_back_jacket_01, bone2Name=BASE_L_back_jacket, swing1LimitDegrees=15, swing2LimitDegrees=5, twistLimitDegrees=5}
 {bone1Name=L_front_jacket_01_dyn, bone2Name=BASE_L_front_jacket, swing1LimitDegrees=10, swing2LimitDegrees=20, twistLimitDegrees=10}
 {bone1Name=L_front_jacket_02_dyn, bone2Name=L_front_jacket_01_dyn, swing1LimitDegrees=10, swing2LimitDegrees=20, twistLimitDegrees=5}
 {bone1Name=L_side_jacket_01, bone2Name=BASE_L_side_jacket, swing1LimitDegrees=15, swing2LimitDegrees=20, twistLimitDegrees=10}
 {bone1Name=R_back_jacket_01, bone2Name=BASE_R_back_jacket, swing1LimitDegrees=15, swing2LimitDegrees=5, twistLimitDegrees=5}
 {bone1Name=R_front_jacket_01, bone2Name=BASE_R_front_jacket, swing1LimitDegrees=10, swing2LimitDegrees=20, twistLimitDegrees=5}
 {bone1Name=R_front_jacket_02, bone2Name=R_front_jacket_01, swing1LimitDegrees=10, swing2LimitDegrees=20, twistLimitDegrees=5}
 {bone1Name=R_side_jacket_01, bone2Name=BASE_R_side_jacket, swing1LimitDegrees=15, swing2LimitDegrees=20, twistLimitDegrees=10}
]
