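# Post-process style cosmetic AnimBP for the female medium VanceGuard body, adding secondary motion to its jacket and bag.
# It copies the attached parent mesh's pose through a Control Rig, then drives six jacket base bones with weighted multi-bone constraints.
# Two chained RigidBody nodes simulate jacket and bag relative to spine_03 and pelvis, with world gravity overridden.
# The EventGraph contains logic for a gravity override, via a deprecated helper, and a stationary-crouch value.
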AnimBlueprint /BRCosmetics/Characters/Player/Female/Medium/Bodies/F_MED_VanceGuard/Meshes/F_MED_VanceGuard_AnimBP
target_skeleton=&/F_MED_VanceGuard_Skeleton
graphs [
 AnimGraph {nodes [
   n1 = Control Rig(Source=Copy Pose From Mesh() {bUseAttachedParent=true}) {
    alphaBoolBlend {blendOption=Linear}
    controlRigAssetReference {blueprintRigClass=&/F_MED_VanceGuard_CtrlRig.F_MED_VanceGuard_CtrlRig_C}
   }
   n2 = "Constraint - BASE_L_front_jacket (PPPPPP)"!(
    ComponentPose=Local To Component(LocalPose=@n1)
    Alpha=1
    ConstraintWeights_0=0.0634
    ConstraintWeights_1:float=0.0209
    ConstraintWeights_2:float=0.4983
    ConstraintWeights_3:float=0.0123
    ConstraintWeights_4:float=0.0122
    ConstraintWeights_5:float=0.3928
   ) -> (Pose) {
    alphaBoolBlend {blendOption=Linear}
    boneToModify {boneName=BASE_L_front_jacket}
    constraintSetup [
     {targetBone {boneName=spine_05}}
     {targetBone {boneName=clavicle_l}}
     {targetBone {boneName=spine_04}}
     {targetBone {boneName=spine_03}}
     {targetBone {boneName=spine_02}}
     {targetBone {boneName=deform_pec_l}}
    ]
    constraintWeights [1,1,1,1,1,1]
   }
   n3 = "Constraint - BASE_L_side_jacket (PPPPPPP)"!(
    ComponentPose=@n2
    Alpha=1
    ConstraintWeights_0=0.0977
    ConstraintWeights_1:float=0.0888
    ConstraintWeights_2:float=0.5577
    ConstraintWeights_3:float=0.067
    ConstraintWeights_4:float=0.0741
    ConstraintWeights_5:float=0.0957
    ConstraintWeights_6:float=0.0189
   ) -> (Pose) {
    alphaBoolBlend {blendOption=Linear}
    boneToModify {boneName=BASE_L_side_jacket}
    constraintSetup [
     {targetBone {boneName=spine_05}}
     {targetBone {boneName=clavicle_l}}
     {targetBone {boneName=spine_04}}
     {targetBone {boneName=upperarm_twist_01_l}}
     {targetBone {boneName=spine_03}}
     {targetBone {boneName=spine_02}}
     {targetBone {boneName=spine_01}}
    ]
    constraintWeights [1,1,1,1,1,1,1]
   }
   n4 = "Constraint - BASE_L_back_jacket (PPPPPPPP)"!(
    ComponentPose=@n3
    Alpha=1
    ConstraintWeights_0=0.2017
    ConstraintWeights_1:float=0.1799
    ConstraintWeights_2:float=0.0304
    ConstraintWeights_3:float=0.3376
    ConstraintWeights_4:float=0.0294
    ConstraintWeights_5:float=0.0875
    ConstraintWeights_6:float=0.1039
    ConstraintWeights_7:float=0.0296
   ) -> (Pose) {
    alphaBoolBlend {blendOption=Linear}
    boneToModify {boneName=BASE_L_back_jacket}
    constraintSetup [
     {targetBone {boneName=spine_05}}
     {targetBone {boneName=clavicle_l}}
     {targetBone {boneName=clavicle_r}}
     {targetBone {boneName=spine_04}}
     {targetBone {boneName=upperarm_twist_01_l}}
     {targetBone {boneName=spine_03}}
     {targetBone {boneName=spine_02}}
     {targetBone {boneName=spine_01}}
    ]
    constraintWeights [1,1,1,1,1,1,1,1]
   }
   n5 = "Constraint - BASE_R_front_jacket (PPPPPP)"!(
    ComponentPose=@n4
    Alpha=1
    ConstraintWeights_0=0.0634
    ConstraintWeights_1:float=0.0209
    ConstraintWeights_2:float=0.4983
    ConstraintWeights_3:float=0.0123
    ConstraintWeights_4:float=0.0122
    ConstraintWeights_5:float=0.3928
   ) -> (Pose) {
    alphaBoolBlend {blendOption=Linear}
    boneToModify {boneName=BASE_R_front_jacket}
    constraintSetup [
     {targetBone {boneName=spine_05}}
     {targetBone {boneName=clavicle_r}}
     {targetBone {boneName=spine_04}}
     {targetBone {boneName=spine_03}}
     {targetBone {boneName=spine_02}}
     {targetBone {boneName=deform_pec_r}}
    ]
    constraintWeights [1,1,1,1,1,1]
   }
   n6 = "Constraint - BASE_R_side_jacket (PPPPPPP)"!(
    ComponentPose=@n5
    Alpha=1
    ConstraintWeights_0=0.0977
    ConstraintWeights_1:float=0.0888
    ConstraintWeights_2:float=0.5577
    ConstraintWeights_3:float=0.067
    ConstraintWeights_4:float=0.0741
    ConstraintWeights_5:float=0.0957
    ConstraintWeights_6:float=0.0189
   ) -> (Pose) {
    alphaBoolBlend {blendOption=Linear}
    boneToModify {boneName=BASE_R_side_jacket}
    constraintSetup [
     {targetBone {boneName=spine_05}}
     {targetBone {boneName=clavicle_r}}
     {targetBone {boneName=spine_04}}
     {targetBone {boneName=upperarm_twist_01_r}}
     {targetBone {boneName=spine_03}}
     {targetBone {boneName=spine_02}}
     {targetBone {boneName=spine_01}}
    ]
    constraintWeights [1,1,1,1,1,1,1]
   }
   n7 = "Constraint - BASE_R_back_jacket (PPPPPPPP)"!(
    ComponentPose=@n6
    Alpha=1
    ConstraintWeights_0=0.2017
    ConstraintWeights_1:float=0.0304
    ConstraintWeights_2:float=0.1799
    ConstraintWeights_3:float=0.3376
    ConstraintWeights_4:float=0.0294
    ConstraintWeights_5:float=0.0875
    ConstraintWeights_6:float=0.1039
    ConstraintWeights_7:float=0.0296
   ) -> (Pose) {
    alphaBoolBlend {blendOption=Linear}
    boneToModify {boneName=BASE_R_back_jacket}
    constraintSetup [
     {targetBone {boneName=spine_05}}
     {targetBone {boneName=clavicle_l}}
     {targetBone {boneName=clavicle_r}}
     {targetBone {boneName=spine_04}}
     {targetBone {boneName=upperarm_twist_01_r}}
     {targetBone {boneName=spine_03}}
     {targetBone {boneName=spine_02}}
     {targetBone {boneName=spine_01}}
    ]
    constraintWeights [1,1,1,1,1,1,1,1]
   }
   n8 = RigidBody!(ComponentPose=@n7, bAlphaBoolEnabled:bool, OverrideWorldGravity:Vector="0,0,0", ExternalForce) -> (Pose) {
    alphaInputType=Bool
    bOverrideWorldGravity=true
    overridePhysicsAsset=&/F_MED_VanceGuard_Physics_jacket_02
    simulationSpace=BaseBoneSpace
    alphaBoolBlend {blendInTime=0.2, blendOption=Linear, blendOutTime=0.2}
    baseBoneRef {boneName=spine_03}
    componentAppliedLinearAccClamp=<0 0 0>
    simSpaceSettings {
     maxAngularAcceleration=1000
     maxAngularVelocity=12
     maxLinearAcceleration=1000
     maxLinearVelocity=800
     velocityScaleZ=0.5
     worldAlpha=0.7
     externalLinearDragV=<0.5 0.5 0.5>
    }
   }
   n9 = RigidBody(ComponentPose=@n8, OverrideWorldGravity="0,0,0") {
    alphaInputType=Bool
    bOverrideWorldGravity=true
    overridePhysicsAsset=&/F_MED_VanceGuard_Physics_bag
    simulationSpace=BaseBoneSpace
    alphaBoolBlend {blendInTime=0.2, blendOption=Linear, blendOutTime=0.2}
    baseBoneRef {boneName=pelvis}
    componentAppliedLinearAccClamp=<0 0 0>
    simSpaceSettings {
     maxAngularAcceleration=1000
     maxAngularVelocity=12
     maxLinearAcceleration=1000
     maxLinearVelocity=800
     velocityScaleZ=0
     worldAlpha=0.7
     externalLinearDragV=<0.5 0.5 0.5>
    }
   }
   Output Pose(Result=Component To Local(ComponentPose=@n9)) {layerGroup=None}
  ]}
 EventGraph {nodes [
   Event BlueprintUpdateAnimation()
   > n1 = Sequence()
   n2 = Gravity Override State Function DEPRECATED(
    execute=@n1.then_1
    CharacterPartAnimInstance=Self-Reference()
    GameTimeInSeconds=GetGameTimeInSeconds()
    GravityOverrideParams=Get Gravity Override Params()
   )
   > Set Gravity Override(Gravity Override=@n2)
   Reroute Node()
   n3 = NOT Boolean(A=Get bIsMoving2D())
   n4 = Get bIsCrouching()
   > Set static_crouching(static_crouching:double="To Float (Boolean)"(InBool=AND Boolean(A=@n4, B=@n3))) -> (Output_Get:double)
  ]}
]
variables [
 Gravity Override {type=Vector}
 Gravity Override Params {type=GravityOverrideParamsStruct}
 static_crouching {type=number}
]
