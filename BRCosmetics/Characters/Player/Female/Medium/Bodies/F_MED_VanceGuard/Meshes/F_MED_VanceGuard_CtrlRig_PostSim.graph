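# Post-sim Control Rig for the female medium VanceGuard body, applying corrective bone adjustments.
# The executed chain shrinks extra_spine_04, then parent-constrains each kneepad to a weighted mix of thigh, calf and knee deform bones.
# Spherical pose readers on the calves feed interpolated weights that offset and rotate the kneepads.
# Crouch-driven bag adjustments and a rotation-remapped knife offset are present but not connected to the execution chain.
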ControlRigBlueprint /BRCosmetics/Characters/Player/Female/Medium/Bodies/F_MED_VanceGuard/Meshes/F_MED_VanceGuard_CtrlRig_PostSim
graphs [
 RigVMFunctionLibrary {links [], nodes []}
 RigVMModel {nodes [
   RigUnit_BeginExecution = &RigUnit_BeginExecution()
   /Script/ControlRig.RigUnit_SetRelativeTransformForItem(
    Child=(Type=Bone,Name="extra_spine_04")
    Parent=(Type=Bone,Name="extra_spine_04")
    Value=(Rotation=(X=0,Y=0,Z=0,W=1),Translation=(X=0,Y=0,Z=0),Scale3D=(X=0.3,Y=0.3,Z=0.3))
    Weight=1
    bParentInitial=False
    bPropagateToChildren=True
    ExecutePin=@RigUnit_BeginExecution.ExecutePin
   ) {
    is=struct RigVMUnitNode
    template="Set Relative Transform::Execute(in Child,in Parent,in bParentInitial,in Value,in Weight,in bPropagateToChildren)"
   }
   > &RigUnit_ParentConstraint(
    AdvancedSettings=(InterpolationType=Average,RotationOrderForFilter=XZY)
    Child=(Type=Bone,Name="kneepad_l")
    Filter=(TranslationFilter=(bX=True,bY=True,bZ=True),RotationFilter=(bX=True,bY=True,bZ=True),ScaleFilter=(bX=false,bY=false,bZ=false))
    Parents=((Item=(Type=Bone,Name="thigh_l"),Weight=0.2115),(Item=(Type=Bone,Name="calf_l"),Weight=0.2437),(Item=(Type=Bone,Name="deform_knee_out_l"),Weight=0.5446),(Item=(Type=Bone,Name="deform_knee_in_l"),Weight=0.0001))
    Weight=1
   )
   > &RigUnit_ParentConstraint(
    AdvancedSettings=(InterpolationType=Average,RotationOrderForFilter=XZY)
    Child=(Type=Bone,Name="kneepad_r")
    Filter=(TranslationFilter=(bX=True,bY=True,bZ=True),RotationFilter=(bX=True,bY=True,bZ=True),ScaleFilter=(bX=False,bY=False,bZ=False))
    Parents=((Item=(Type=Bone,Name="thigh_r"),Weight=0.2115),(Item=(Type=Bone,Name="calf_r"),Weight=0.2437),(Item=(Type=Bone,Name="deform_knee_out_r"),Weight=0.5446),(Item=(Type=Bone,Name="deform_knee_in_r"),Weight=0.0001))
    Weight=1
   )
   > n1 = &RigUnit_SphericalPoseReader(
    ActiveRegionScaleFactors=(PositiveWidth=0.8,NegativeWidth=0.8,PositiveHeight=0.8,NegativeHeight=0.8)
    ActiveRegionSize=0.2
    Debug=(bDrawDebug=false,bDraw2D=false,bDrawLocalAxes=false,DebugScale=30,DebugSegments=30,DebugThickness=0.2)
    DriverAxis=(X=-1,Y=0,Z=0)
    DriverItem=(Type=Bone,Name="calf_r")
    FalloffRegionScaleFactors=(PositiveWidth=0.5,NegativeWidth=0.5,PositiveHeight=0.5,NegativeHeight=0.5)
    FalloffSize=0.5
    OptionalParentItem=(Type=Bone,Name="None")
    RotationOffset=(X=-90,Y=0,Z=50)
   )
   n2 = &RigVMFunction_AlphaInterp(InterpSpeedDecreasing=10, InterpSpeedIncreasing=10, Scale=1, bClampResult=False, Value=@n1.OutputParam)
   > &RigUnit_SetRelativeTransformForItem(
    Child=(Type=Bone,Name="kneepad_r")
    Parent=(Type=Bone,Name="kneepad_r")
    Value=(Rotation=(X=0,Y=0,Z=0.1736,W=0.9848),Translation=(X=1,Y=1,Z=0),Scale3D=(X=1,Y=1,Z=1))
    bParentInitial=False
    bPropagateToChildren=True
    Weight=@n2.Result
   ) {template="Set Relative Transform::Execute(in Child,in Parent,in bParentInitial,in Value,in Weight,in bPropagateToChildren)"}
   > n3 = &RigUnit_SphericalPoseReader(
    ActiveRegionScaleFactors=(PositiveWidth=0.8,NegativeWidth=0.8,PositiveHeight=0.8,NegativeHeight=0.8)
    ActiveRegionSize=0.2
    Debug=(bDrawDebug=false,bDraw2D=false,bDrawLocalAxes=false,DebugScale=30,DebugSegments=30,DebugThickness=0.2)
    DriverAxis=(X=1,Y=0,Z=0)
    DriverItem=(Type=Bone,Name="calf_l")
    FalloffRegionScaleFactors=(PositiveWidth=0.5,NegativeWidth=0.5,PositiveHeight=0.5,NegativeHeight=0.5)
    FalloffSize=0.5
    OptionalParentItem=(Type=Bone,Name="None")
    RotationOffset=(X=90,Y=0,Z=50)
   )
   n4 = &RigVMFunction_AlphaInterp(InterpSpeedDecreasing=10, InterpSpeedIncreasing=10, Scale=1, bClampResult=False, Value=@n3.OutputParam)
   > &RigUnit_SetRelativeTransformForItem(
    Child=(Type=Bone,Name="kneepad_l")
    Parent=(Type=Bone,Name="kneepad_l")
    Value=(Rotation=(X=0,Y=0,Z=0.1736,W=0.9848),Translation=(X=-1,Y=-1,Z=0),Scale3D=(X=1,Y=1,Z=1))
    bParentInitial=False
    bPropagateToChildren=True
    Weight=@n4.Result
   ) {template="Set Relative Transform::Execute(in Child,in Parent,in bParentInitial,in Value,in Weight,in bPropagateToChildren)"}
   n5 = &RigVMFunction_AlphaInterp(InterpSpeedDecreasing=10, InterpSpeedIncreasing=10, Scale=1, bClampResult=False, Value=RigVMVariableNode() {variable=static_crouching}.Value)
   &RigUnit_SetRelativeTransformForItem(
    Child=(Type=Bone,Name="BASE_bag")
    ExecutePin=()
    Parent=(Type=Bone,Name="BASE_bag")
    Value=(Rotation=(X=0,Y=0.0872,Z=0,W=0.9962),Translation=(X=0,Y=0,Z=0),Scale3D=(X=1,Y=1,Z=1))
    bParentInitial=False
    bPropagateToChildren=True
    Weight=@n5.Result
   ) {template="Set Relative Transform::Execute(in Child,in Parent,in bParentInitial,in Value,in Weight,in bPropagateToChildren)"}
   > &RigUnit_SetRelativeTransformForItem(
    Child=(Type=Bone,Name="rotate_2")
    Parent=(Type=Bone,Name="rotate_2")
    Value=(Rotation=(X=-0.0872,Y=0,Z=0,W=0.9962),Translation=(X=0,Y=0,Z=0),Scale3D=(X=1,Y=1,Z=1))
    bParentInitial=False
    bPropagateToChildren=True
    Weight=@n5.Result
   ) {template="Set Relative Transform::Execute(in Child,in Parent,in bParentInitial,in Value,in Weight,in bPropagateToChildren)"}
   RigVMFunction_MathQuaternionToEuler = &RigVMFunction_MathQuaternionToEuler(
    RotationOrder=ZYX
    Value=&RigUnit_GetTransform(Item=(Type=Bone,Name="rotate_1"), Space=LocalSpace, bInitial=False).Transform.Rotation
   )
   &RigUnit_ModifyTransforms(
    ExecutePin=()
    ItemToModify=((Item=(Type=Bone,Name="BASE_knif1"),Transform=(Rotation=(X=0,Y=0,Z=0,W=1),Translation=(X=-0.25,Y=1,Z=0),Scale3D=(X=1,Y=1,Z=1))))
    Weight=&RigVMFunction_MathDoubleRemap(SourceMaximum=-33, SourceMinimum=-10, TargetMaximum=1, TargetMinimum=0, Value=@RigVMFunction_MathQuaternionToEuler.Result.X).Result
   )
  ]}
]
hierarchy [
 root {loc=<0 0 0>, rot=<0 0 0 1>}
 pelvis {parent=root, loc=<0 -0.5301 98.0571>, rot=<0.0127 0.707 -0.0127 -0.707>}
 spine_01 {parent=pelvis, loc=<6.1208 -0.9328 0>, rot=<0 0 0.034 -0.9994>}
 spine_02 {parent=spine_01, loc=<6.5997 -0.2095 0>, rot=<0 0 0.0238 -0.9997>}
 spine_03 {parent=spine_02, loc=<6.4914 -0.1237 0>, rot=<0 0 0.0149 -0.9999>}
 spine_04 {parent=spine_03, loc=<6.9598 0.0794 0>, rot=<0 0 -0.0132 -0.9999>}
 spine_05 {parent=spine_04, loc=<6.9987 0 0>, rot=<0 0 -0.0015 -1>}
 neck_01 {parent=spine_05, loc=<11.3138 0.1529 0>, rot=<0 0 -0.2003 -0.9797>}
 neck_02 {parent=neck_01, loc=<4.5627 0 0>, rot=<0 0 0.0504 -0.9987>}
 head {parent=neck_02, loc=<5.3914 0 0>, rot=<0 0 0.115 -0.9934>}
 head_end {parent=head, loc=<9.2548 0 0>, rot=<0 0 0 1>}
 clavicle_l {parent=spine_05, loc=<5.4831 -0.1053 -3.2946>, rot=<0.0465 -0.7289 -0.009 -0.6829>}
 upperarm_l {parent=clavicle_l, loc=<9.8658 -1.2136 -0.4105>, rot=<0.0164 -0.3245 -0.0269 -0.9454>}
 lowerarm_l {parent=upperarm_l, loc=<26.465 -0.8345 0.259>, rot=<0.0036 0.0161 -0.2298 -0.9731>}
 hand_l {parent=lowerarm_l, loc=<22.7355 -0.3203 -0.6105>, rot=<-0.6784 -0.0223 0.0735 -0.7307>}
 thumb_01_l {parent=hand_l, loc=<1.3118 -0.3191 -2.459>, rot=<0.6546 -0.4042 0.0664 -0.6354>}
 thumb_02_l {parent=thumb_01_l, loc=<5.9336 -0.0448 0.0401>, rot=<-0.0037 0.0213 0.298 -0.9543>}
 thumb_03_l {parent=thumb_02_l, loc=<2.2223 0.3682 -0.0907>, rot=<0.0118 -0.013 0.0066 -0.9998>}
 index_metacarpal_l {parent=hand_l, loc=<1.6535 0.0183 -2.96>, rot=<0.0158 -0.0124 -0.0173 -0.9996>}
 index_01_l {parent=index_metacarpal_l, loc=<8.3008 0.0357 -0.0524>, rot=<-0.0044 0.0485 0.0924 -0.9945>}
 index_02_l {parent=index_01_l, loc=<3.3587 -0.3004 0.0452>, rot=<-0.002 0.0027 0.3225 -0.9466>}
 index_03_l {parent=index_02_l, loc=<2.5578 -0.123 -0.0252>, rot=<-0.002 0.0067 0.2558 -0.9667>}
 middle_metacarpal_l {parent=hand_l, loc=<2.1856 -0.1725 -0.5984>, rot=<0.006 0.0168 -0.0329 -0.9993>}
 middle_01_l {parent=middle_metacarpal_l, loc=<7.3275 -0.2447 0.0129>, rot=<-0.0208 0.0042 0.0858 -0.9961>}
 middle_02_l {parent=middle_01_l, loc=<4.0123 0.1401 0.1175>, rot=<-0.0218 0.0258 0.3512 -0.9357>}
 middle_03_l {parent=middle_02_l, loc=<2.6077 0.0534 0.1373>, rot=<0.0082 0.0188 0.1708 -0.9851>}
 ring_metacarpal_l {parent=hand_l, loc=<2.2074 -0.3625 1.3638>, rot=<-0.0284 0.0722 -0.0458 -0.9959>}
 ring_01_l {parent=ring_metacarpal_l, loc=<7.0295 -0.6197 -0.1446>, rot=<-0.0489 -0.0252 0.1892 -0.9804>}
 ring_02_l {parent=ring_01_l, loc=<3.6174 0.4647 -0.1027>, rot=<0 0 0.2806 -0.9598>}
 ring_03_l {parent=ring_02_l, loc=<2.4152 0.114 0.0199>, rot=<0 0 0.1189 -0.9929>}
 pinky_metacarpal_l {parent=hand_l, loc=<1.7616 -0.9753 3.3814>, rot=<-0.0397 0.0814 -0.069 -0.9935>}
 pinky_01_l {parent=pinky_metacarpal_l, loc=<7.1709 -0.4831 -0.1094>, rot=<-0.0654 -0.0035 0.2242 -0.9723>}
 pinky_02_l {parent=pinky_01_l, loc=<2.8274 0.0893 -0.1043>, rot=<0.0037 0.0011 0.2854 -0.9584>}
 pinky_03_l {parent=pinky_02_l, loc=<2.2046 0.0307 -0.0078>, rot=<-0.0002 0.0098 0.1445 -0.9895>}
 deform_wrist_dn_base_l {parent=hand_l, loc=<0 -1.5326 0>, rot=<0 0 0 -1>}
 deform_wrist_dn_l {parent=deform_wrist_dn_base_l, loc=<0 0 0>, rot=<0 0 0 1>}
 deform_wrist_up_base_l {parent=hand_l, loc=<0 2.3009 0>, rot=<0 0 0 -1>}
 deform_wrist_up_l {parent=deform_wrist_up_base_l, loc=<0 0 0>, rot=<0 0 0 1>}
 lowerarm_twist_02_l {parent=lowerarm_l, loc=<10.6707 0 0>, rot=<0 0 0 1>}
 lowerarm_twist_01_l {parent=lowerarm_l, loc=<18.316 0 0>, rot=<0 0 0 1>}
 deform_elbow_in_base_l {parent=lowerarm_l, loc=<0.1834 3.6062 -0.2572>, rot=<0 0 0 -1>}
 deform_elbow_in_l {parent=deform_elbow_in_base_l, loc=<0 0 0>, rot=<0 0 0 1>}
 deform_elbow_out_base_l {parent=lowerarm_l, loc=<0 -4 0>, rot=<0 0 0 1>}
 deform_elbow_out_l {parent=deform_elbow_out_base_l, loc=<0 0 0>, rot=<0 0 0 1>}
 upperarm_twist_01_l {parent=upperarm_l, loc=<7.2135 -0.1134 0>, rot=<0 0 0 1>}
 deform_upperarm_bk_base_l {parent=upperarm_twist_01_l, loc=<-6.4053 -4.3746 0.314>, rot=<0 0 0 -1>}
 deform_upperarm_bk_l {parent=deform_upperarm_bk_base_l, loc=<0 0 0>, rot=<0 0 0 1>}
 deform_upperarm_fr_base_l {parent=upperarm_twist_01_l, loc=<-5.6312 5.092 0.3354>, rot=<0 0 0 -1>}
 deform_upperarm_fr_l {parent=deform_upperarm_fr_base_l, loc=<0 0 0>, rot=<0 0 0 1>}
 deform_upperarm_up_base_l {parent=upperarm_twist_01_l, loc=<-8.891 0.4825 3.9517>, rot=<0 0 0 -1>}
 deform_upperarm_up_l {parent=deform_upperarm_up_base_l, loc=<0 0 0>, rot=<0 0 0 1>}
 upperarm_twist_02_l {parent=upperarm_l, loc=<19.7998 -0.1134 0>, rot=<0 0 0 1>}
 BASE_binoculars {parent=upperarm_l, loc=<5.2013 -0.5502 4.3269>, rot=<0 0 0 1>}
 binoculars {parent=BASE_binoculars, loc=<0 0 0>, rot=<0 0 0 1>}
 clavicle_r {parent=spine_05, loc=<5.4831 -0.1053 3.2946>, rot=<0.7289 0.0465 -0.6829 0.009>}
 upperarm_r {parent=clavicle_r, loc=<-9.8658 1.2136 0.4105>, rot=<0.0164 -0.3245 -0.0269 -0.9454>}
 lowerarm_r {parent=upperarm_r, loc=<-26.465 0.8345 -0.259>, rot=<0.0036 0.0161 -0.2298 -0.9731>}
 hand_r {parent=lowerarm_r, loc=<-22.7355 0.3203 0.6106>, rot=<-0.6784 -0.0223 0.0735 -0.7307>}
 thumb_01_r {parent=hand_r, loc=<-1.3118 0.3191 2.459>, rot=<0.6546 -0.4042 0.0664 -0.6354>}
 thumb_02_r {parent=thumb_01_r, loc=<-5.9336 0.0448 -0.0401>, rot=<-0.0037 0.0213 0.298 -0.9543>}
 thumb_03_r {parent=thumb_02_r, loc=<-2.2223 -0.3682 0.0907>, rot=<0.0118 -0.013 0.0066 -0.9998>}
 index_metacarpal_r {parent=hand_r, loc=<-1.6535 -0.0183 2.96>, rot=<0.0158 -0.0124 -0.0173 -0.9996>}
 index_01_r {parent=index_metacarpal_r, loc=<-8.3008 -0.0357 0.0524>, rot=<-0.0044 0.0485 0.0924 -0.9945>}
 index_02_r {parent=index_01_r, loc=<-3.3587 0.3004 -0.0452>, rot=<-0.002 0.0027 0.3225 -0.9466>}
 index_03_r {parent=index_02_r, loc=<-2.5578 0.123 0.0252>, rot=<-0.002 0.0067 0.2558 -0.9667>}
 middle_metacarpal_r {parent=hand_r, loc=<-2.1856 0.1725 0.5984>, rot=<0.006 0.0168 -0.0329 -0.9993>}
 middle_01_r {parent=middle_metacarpal_r, loc=<-7.3275 0.2447 -0.0129>, rot=<-0.0208 0.0042 0.0858 -0.9961>}
 middle_02_r {parent=middle_01_r, loc=<-4.0123 -0.1401 -0.1175>, rot=<-0.0218 0.0258 0.3512 -0.9357>}
 middle_03_r {parent=middle_02_r, loc=<-2.6077 -0.0534 -0.1373>, rot=<0.0082 0.0188 0.1708 -0.9851>}
 ring_metacarpal_r {parent=hand_r, loc=<-2.2074 0.3625 -1.3638>, rot=<-0.0284 0.0722 -0.0458 -0.9959>}
 ring_01_r {parent=ring_metacarpal_r, loc=<-7.0295 0.6197 0.1446>, rot=<-0.0489 -0.0252 0.1892 -0.9804>}
 ring_02_r {parent=ring_01_r, loc=<-3.6174 -0.4647 0.1027>, rot=<0 0 0.2806 -0.9598>}
 ring_03_r {parent=ring_02_r, loc=<-2.4152 -0.114 -0.0199>, rot=<0 0 0.1189 -0.9929>}
 pinky_metacarpal_r {parent=hand_r, loc=<-1.7616 0.9753 -3.3814>, rot=<-0.0397 0.0814 -0.069 -0.9935>}
 pinky_01_r {parent=pinky_metacarpal_r, loc=<-7.1709 0.4831 0.1094>, rot=<-0.0654 -0.0035 0.2242 -0.9723>}
 pinky_02_r {parent=pinky_01_r, loc=<-2.8274 -0.0893 0.1043>, rot=<0.0037 0.0011 0.2854 -0.9584>}
 pinky_03_r {parent=pinky_02_r, loc=<-2.2046 -0.0307 0.0078>, rot=<-0.0002 0.0098 0.1445 -0.9895>}
 deform_wrist_dn_base_r {parent=hand_r, loc=<-0.0001 1.5328 0>, rot=<0 0 0 -1>}
 deform_wrist_dn_r {parent=deform_wrist_dn_base_r, loc=<0 0 0>, rot=<0 0 0 1>}
 deform_wrist_up_base_r {parent=hand_r, loc=<-0.0002 -2.3007 -0.0001>, rot=<0 0 0 -1>}
 deform_wrist_up_r {parent=deform_wrist_up_base_r, loc=<0 0 0>, rot=<0 0 0 1>}
 lowerarm_twist_02_r {parent=lowerarm_r, loc=<-10.6707 0 0>, rot=<0 0 0 -1>}
 lowerarm_twist_01_r {parent=lowerarm_r, loc=<-18.316 0 0>, rot=<0 0 0 -1>}
 deform_elbow_in_base_r {parent=lowerarm_r, loc=<-0.1834 -3.6062 0.2572>, rot=<0 0 0 -1>}
 deform_elbow_in_r {parent=deform_elbow_in_base_r, loc=<0 0 0>, rot=<0 0 0 1>}
 deform_elbow_out_base_r {parent=lowerarm_r, loc=<0 4 -0.0001>, rot=<0 0 0 -1>}
 deform_elbow_out_r {parent=deform_elbow_out_base_r, loc=<0 0 0>, rot=<0 0 0 1>}
 upperarm_twist_01_r {parent=upperarm_r, loc=<-7.2135 0.1134 0>, rot=<0 0 0 -1>}
 deform_upperarm_bk_base_r {parent=upperarm_twist_01_r, loc=<6.4053 4.3746 -0.314>, rot=<0 0 0 -1>}
 deform_upperarm_bk_r {parent=deform_upperarm_bk_base_r, loc=<0 0 0>, rot=<0 0 0 1>}
 deform_upperarm_fr_base_r {parent=upperarm_twist_01_r, loc=<5.6312 -5.092 -0.3354>, rot=<0 0 0 -1>}
 deform_upperarm_fr_r {parent=deform_upperarm_fr_base_r, loc=<0 0 0>, rot=<0 0 0 1>}
 deform_upperarm_up_base_r {parent=upperarm_twist_01_r, loc=<8.891 -0.4825 -3.9517>, rot=<0 0 0 -1>}
 deform_upperarm_up_r {parent=deform_upperarm_up_base_r, loc=<0 0 0>, rot=<0 0 0 1>}
 upperarm_twist_02_r {parent=upperarm_r, loc=<-19.7998 0.1134 0>, rot=<0 0 0 -1>}
 attach_backpack {parent=spine_05, loc=<-0.0953 -13.4502 -0.0046>, rot=<-0.0093 -0.707 0.0092 -0.7071>}
 Backpack_BR {parent=attach_backpack, type=Null, loc=<0 3.9199 -3.4812>, rot=<0 0 0 1>}
 attach_cape {parent=spine_05, loc=<9.0354 -11.7004 -0.0076>, rot=<0.0616 -0.7044 -0.0617 -0.7044>}
 cape {parent=attach_cape, type=Null, loc=<0 0 0>, rot=<0 0 0 1>}
 BASE_collar_m {parent=spine_05, loc=<9.991 -6.7419 0>, rot=<0.0002 -0.002 0.0198 -0.9998>}
 collar_m {parent=BASE_collar_m, loc=<0 0 0>, rot=<0 0 0 1>}
 BASE_collar_l4_l {parent=spine_05, loc=<10.0123 -0.3596 -7.9517>, rot=<0.1478 -0.6144 0.7745 -0.0279>}
 collar_l4_l {parent=BASE_collar_l4_l, loc=<0 0 0>, rot=<0 0 0 1>}
 BASE_collar_l3_l {parent=spine_05, loc=<10.3773 -2.6667 -7.3731>, rot=<0.0901 -0.5721 0.8149 0.0217>}
 collar_l3_l {parent=BASE_collar_l3_l, loc=<0 0 0>, rot=<0 0 0 1>}
 BASE_collar_l2_l {parent=spine_05, loc=<10.1085 -4.6062 -5.8197>, rot=<0.04 -0.3145 0.9482 0.0218>}
 collar_l2_l {parent=BASE_collar_l2_l, loc=<0 0 0>, rot=<0 0 0 1>}
 BASE_collar_l1_l {parent=spine_05, loc=<10.0267 -6.2033 -3.0844>, rot=<0.0377 -0.187 0.9814 0.0199>}
 collar_l1_l {parent=BASE_collar_l1_l, loc=<0 0 0>, rot=<0 0 0 1>}
 BASE_collar_r4_r {parent=spine_05, loc=<10.0123 -0.3596 7.9517>, rot=<0.6143 0.1504 -0.03 -0.774>}
 collar_r4_r {parent=BASE_collar_r4_r, loc=<0 0 0>, rot=<0 0 0 1>}
 BASE_collar_r3_r {parent=spine_05, loc=<10.3773 -2.6667 7.3731>, rot=<0.5226 0.0984 0.0119 -0.8468>}
 collar_r3_r {parent=BASE_collar_r3_r, loc=<0 0 0>, rot=<0 0 0 1>}
 BASE_collar_r2_r {parent=spine_05, loc=<10.1085 -4.6062 5.8197>, rot=<0.3144 0.0365 0.023 -0.9483>}
 collar_r2_r {parent=BASE_collar_r2_r, loc=<0 0 0>, rot=<0 0 0 1>}
 BASE_collar_r1_r {parent=spine_05, loc=<10.0267 -6.2033 3.0844>, rot=<0.187 0.0377 0.0199 -0.9814>}
 collar_r1_r {parent=BASE_collar_r1_r, loc=<0 0 0>, rot=<0 0 0 1>}
 BASE_collar_l5_l {parent=spine_05, loc=<9.0663 1.6604 -7.6292>, rot=<0.1942 -0.7645 0.6127 -0.0498>}
 collar_l5_l {parent=BASE_collar_l5_l, loc=<0 0 0>, rot=<0 0 0 1>}
 BASE_collar_l6_l {parent=spine_05, loc=<8.2461 4.0111 -6.8156>, rot=<0.2307 -0.8226 0.5135 -0.0798>}
 collar_l6_l {parent=BASE_collar_l6_l, loc=<0 0 0>, rot=<0 0 0 1>}
 BASE_collar_l7_l {parent=spine_05, loc=<7.1595 6.1923 -4.8084>, rot=<0.1782 -0.919 0.3393 -0.0929>}
 collar_l7_l {parent=BASE_collar_l7_l, loc=<0 0 0>, rot=<0 0 0 1>}
 BASE_collar_r7_r {parent=spine_05, loc=<7.1598 6.1924 4.8084>, rot=<0.938 0.1766 -0.0892 -0.2845>}
 collar_r7_r {parent=BASE_collar_r7_r, loc=<0 0 0>, rot=<0 0 0 1>}
 BASE_collar_r6_r {parent=spine_05, loc=<8.2461 4.0111 6.8156>, rot=<0.8226 0.2307 -0.0798 -0.5135>}
 collar_r6_r {parent=BASE_collar_r6_r, loc=<0 0 0>, rot=<0 0 0 1>}
 BASE_collar_r5_r {parent=spine_05, loc=<9.0661 1.6604 7.6292>, rot=<0.7645 0.1942 -0.0498 -0.6127>}
 collar_r5_r {parent=BASE_collar_r5_r, loc=<0 0 0>, rot=<0 0 0 1>}
 deform_pec_aim_base_l {parent=spine_04, loc=<3.2456 9.3309 -1.2237>, rot=<0.0603 -0.4033 0.3318 -0.8507>}
 deform_pec_aim_l {parent=deform_pec_aim_base_l, loc=<0 0 0>, rot=<0 0 0 1>}
 deform_pec_base_l {parent=deform_pec_aim_l, loc=<9.3665 0 0>, rot=<0 0 0 1>}
 deform_pec_l {parent=deform_pec_base_l, loc=<0 0 0>, rot=<0 0 0 1>}
 deform_pec_aim_base_r {parent=spine_04, loc=<3.2452 9.3309 1.2237>, rot=<-0.4033 -0.0603 0.8507 0.3318>}
 deform_pec_aim_r {parent=deform_pec_aim_base_r, loc=<0 0 0>, rot=<0 0 0 1>}
 deform_pec_base_r {parent=deform_pec_aim_r, loc=<-9.3667 -0.0003 -0.0004>, rot=<0 0 0 1>}
 deform_pec_r {parent=deform_pec_base_r, loc=<0 0 0>, rot=<0 0 0 1>}
 extra_spine_04 {parent=spine_04, loc=<0 0 0>, rot=<0 0 0 1>}
 BASE_L_back_jacket {parent=spine_03, loc=<7.6282 -7.8882 -5.2446>, rot=<0.0165 -0.9998 -0.0006 -0.0067>}
 L_back_jacket_01 {parent=BASE_L_back_jacket, loc=<0 0 0>, rot=<0 0 0 1>}
 BASE_R_back_jacket {parent=spine_03, loc=<7.4689 -7.1912 5.2692>, rot=<0.0165 -0.9998 -0.0006 -0.0067>}
 R_back_jacket_01 {parent=BASE_R_back_jacket, loc=<0 0 0>, rot=<0 0 0 1>}
 BASE_R_side_jacket {parent=spine_03, loc=<4.7486 0.2504 11.602>, rot=<0.0038 -0.0701 0.996 -0.0546>}
 R_side_jacket_01 {parent=BASE_R_side_jacket, loc=<0 0 0>, rot=<0 0 0 1>}
 BASE_L_side_jacket {parent=spine_03, loc=<4.7486 0.2504 -11.602>, rot=<0.0547 -0.9977 0.0398 -0.0022>}
 L_side_jacket_01 {parent=BASE_L_side_jacket, loc=<0 0 0>, rot=<0 0 0 1>}
 BASE_L_front_jacket {parent=spine_03, loc=<6.939 10.8518 -9.204>, rot=<0.0426 -0.9373 -0.3427 -0.0462>}
 L_front_jacket_01_dyn {parent=BASE_L_front_jacket, loc=<0 0 0>, rot=<0 0 0 1>}
 L_front_jacket_02_dyn {parent=L_front_jacket_01_dyn, loc=<5.6485 0 0>, rot=<0 0 0 1>}
 BASE_R_front_jacket {parent=spine_03, loc=<6.939 10.8518 9.204>, rot=<0.0291 0.3984 0.9159 -0.0389>}
 R_front_jacket_01 {parent=BASE_R_front_jacket, loc=<0 0 0>, rot=<0 0 0 1>}
 R_front_jacket_02 {parent=R_front_jacket_01, loc=<5.7429 0 0>, rot=<0 0 0 1>}
 thigh_l {parent=pelvis, loc=<-1.0855 -0.2109 -8.7053>, rot=<0.0009 0.0355 0.0155 -0.9993>}
 calf_l {parent=thigh_l, loc=<-42.5284 -0.0019 -0.7297>, rot=<0.0006 -0.0161 -0.0391 -0.9991>}
 foot_l {parent=calf_l, loc=<-46.7401 -2.5014 -0.592>, rot=<0.0427 -0.0238 0.0353 -0.9982>}
 ball_l {parent=foot_l, loc=<-3.5658 10.8066 -0.1544>, rot=<0.0001 -0.0001 -0.7063 -0.7079>}
 calf_twist_02_l {parent=calf_l, loc=<-19.1546 -1.7943 -0.2638>, rot=<0.0237 0.0095 -0.0035 -0.9997>}
 calf_twist_01_l {parent=calf_l, loc=<-35.8483 -1.9197 -0.5794>, rot=<0.0237 0.0095 -0.0035 -0.9997>}
 deform_knee_in_base_l {parent=calf_l, loc=<0.8031 -5.0231 0.2456>, rot=<0 0 0 -1>}
 deform_knee_in_l {parent=deform_knee_in_base_l, loc=<0 0 0>, rot=<0 0 0 1>}
 deform_knee_out_base_l {parent=calf_l, loc=<0 4 0>, rot=<0 0 0 -1>}
 deform_knee_out_l {parent=deform_knee_out_base_l, loc=<0 0 0>, rot=<0 0 0 1>}
 kneepad_l {parent=calf_l, loc=<0 3.9 0>, rot=<0 0 0 -1>}
 thigh_twist_01_l {parent=thigh_l, loc=<-27.952 0 0>, rot=<0.0013 0 0 -1>}
 deform_groin_base_l {parent=thigh_l, loc=<0.4801 5.6776 -0.2486>, rot=<0 0 0 -1>}
 deform_groin_l {parent=deform_groin_base_l, loc=<0 0 0>, rot=<0 0 0 1>}
 thigh_r {parent=pelvis, loc=<-1.0855 -0.2109 8.7053>, rot=<0.0355 -0.0009 0.9993 0.0155>}
 calf_r {parent=thigh_r, loc=<42.5284 0.0019 0.7297>, rot=<0.0006 -0.0161 -0.0391 -0.9991>}
 foot_r {parent=calf_r, loc=<46.7401 2.5014 0.592>, rot=<0.0427 -0.0238 0.0353 -0.9982>}
 ball_r {parent=foot_r, loc=<3.5658 -10.8066 0.1544>, rot=<0.0001 -0.0001 -0.7063 -0.7079>}
 calf_twist_02_r {parent=calf_r, loc=<19.1546 1.7943 0.2638>, rot=<0.0237 0.0095 -0.0035 -0.9997>}
 calf_twist_01_r {parent=calf_r, loc=<35.8483 1.9197 0.5794>, rot=<0.0237 0.0095 -0.0035 -0.9997>}
 deform_knee_in_base_r {parent=calf_r, loc=<-0.8031 5.0231 -0.2457>, rot=<0 0 0 -1>}
 deform_knee_in_r {parent=deform_knee_in_base_r, loc=<0 0 0>, rot=<0 0 0 1>}
 deform_knee_out_base_r {parent=calf_r, loc=<0 -4 0>, rot=<0 0 0 -1>}
 deform_knee_out_r {parent=deform_knee_out_base_r, loc=<0 0 0>, rot=<0 0 0 1>}
 kneepad_r {parent=calf_r, loc=<0 -3.9 0>, rot=<0 0 0 -1>}
 thigh_twist_01_r {parent=thigh_r, loc=<27.952 0 0>, rot=<0.0013 0 0 -1>}
 deform_groin_base_r {parent=thigh_r, loc=<-0.4801 -5.6776 0.2486>, rot=<0 0 0 -1>}
 deform_groin_r {parent=deform_groin_base_r, loc=<0 0 0>, rot=<0 0 0 1>}
 deform_glute_base_l {parent=pelvis, loc=<-7.3466 -9.0213 -9.1245>, rot=<0.0172 0.031 -0.4477 -0.8935>}
 deform_glute_l {parent=deform_glute_base_l, loc=<0 0 0>, rot=<0 0 0 1>}
 deform_glute_base_r {parent=pelvis, loc=<-7.3464 -9.0213 9.1245>, rot=<0.031 -0.0172 0.8935 -0.4477>}
 deform_glute_r {parent=deform_glute_base_r, loc=<0 0 0>, rot=<0 0 0 1>}
 BASE_belt_03 {parent=pelvis, loc=<5.3466 -8.8936 -8.2231>, rot=<-0.0155 -0.9993 0.0009 -0.0355>}
 belt_03 {parent=BASE_belt_03, loc=<0 0 0>, rot=<0 0 0 1>}
 BASE_belt_01 {parent=pelvis, loc=<3.27 6.2236 -8.4146>, rot=<-0.0355 0.0009 0.9993 0.0155>}
 belt_01 {parent=BASE_belt_01, loc=<0 0 0>, rot=<0 0 0 1>}
 BASE_belt_04 {parent=pelvis, loc=<6.0785 8.5165 -0.0003>, rot=<-0.0355 0.0009 0.9993 0.0155>}
 belt_04 {parent=BASE_belt_04, loc=<0 0 0>, rot=<0 0 0 1>}
 BASE_pot_1 {parent=pelvis, loc=<6.0347 8.4813 -1.4676>, rot=<-0.0127 -0.707 0.0127 -0.707>}
 pot_1 {parent=BASE_pot_1, loc=<0 0 0>, rot=<0 0 0 1>}
 BASE_pot_3 {parent=pelvis, loc=<3.3321 6.4268 -9.4531>, rot=<-0.0127 -0.707 0.0127 -0.707>}
 pot_3 {parent=BASE_pot_3, loc=<0 0 0>, rot=<0 0 0 1>}
 BASE_pot_2 {parent=pelvis, loc=<3.3698 7.4581 -5.5543>, rot=<-0.0127 -0.707 0.0127 -0.707>}
 pot_2 {parent=BASE_pot_2, loc=<0 0 0>, rot=<0 0 0 1>}
 BASE_bag {parent=pelvis, loc=<6.0047 -0.9365 -13.1148>, rot=<0.0009 0.0355 0.0155 -0.9993>}
 bag {parent=BASE_bag, loc=<0 0 0>, rot=<0 0 0 1>}
 BASE_belt_02 {parent=bag, loc=<0 0 0>, rot=<0.0587 -0.7047 -0.7047 -0.0587>}
 belt_02 {parent=BASE_belt_02, loc=<0 0 0>, rot=<0 0 0 1>}
 rotate_1 {parent=bag, loc=<0 1.3477 -1.2304>, rot=<0 0 0 1>}
 BASE_knif1 {parent=rotate_1, loc=<-8.7736 3.8701 -0.2102>, rot=<-0.2279 -0.6575 -0.1652 -0.6989>}
 knif1 {parent=BASE_knif1, loc=<0 0 0>, rot=<0 0 0 1>}
 rotate_2 {parent=bag, loc=<-0.0134 -1.3468 -1.2314>, rot=<-0.0049 -1 -0.0004 0>}
 BASE_knif2 {parent=rotate_2, loc=<8.9211 -2.8481 0.9478>, rot=<-0.1561 -0.6781 -0.0894 -0.7126>}
 knif2 {parent=BASE_knif2, loc=<0 0 0>, rot=<0 0 0 1>}
]
variables [static_crouching {type=double, default=0}]
