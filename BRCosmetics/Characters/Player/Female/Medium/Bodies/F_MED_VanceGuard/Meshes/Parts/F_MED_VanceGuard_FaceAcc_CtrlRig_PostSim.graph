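# Post-sim Control Rig for the VanceGuard female face accessory mesh that shrinks the back hair as the head moves into a particular pose.
# A spherical pose reader tracks the head bone against spine_04 using an asymmetric active and falloff region, with debug drawing enabled.
# Its output weights a self-relative scale-down of BASE_back_hair, which propagates to the back hair child bones.
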ControlRigBlueprint /BRCosmetics/Characters/Player/Female/Medium/Bodies/F_MED_VanceGuard/Meshes/Parts/F_MED_VanceGuard_FaceAcc_CtrlRig_PostSim
graphs [
 RigVMFunctionLibrary {links [], nodes []}
 RigVMModel {nodes [
   RigUnit_BeginExecution = &RigUnit_BeginExecution()
   n1 = &RigUnit_SphericalPoseReader(
    ActiveRegionScaleFactors=(PositiveWidth=0.3,NegativeWidth=0.3,PositiveHeight=0.6,NegativeHeight=1)
    ActiveRegionSize=0.15
    Debug=(bDrawDebug=True,bDraw2D=False,bDrawLocalAxes=False,DebugScale=25,DebugSegments=20,DebugThickness=0.25)
    DriverAxis=(X=1,Y=0,Z=0)
    DriverItem=(Type=Bone,Name="head")
    FalloffRegionScaleFactors=(PositiveWidth=0.8,NegativeWidth=1,PositiveHeight=1.3,NegativeHeight=1)
    FalloffSize=0.4
    FlipHeightScaling=false
    FlipWidthScaling=False
    OptionalParentItem=(Type=Bone,Name="spine_04")
    RotationOffset=(X=0,Y=-62,Z=-90)
    ExecutePin=@RigUnit_BeginExecution.ExecutePin
   )
   > /Script/ControlRig.RigUnit_SetRelativeTransformForItem(
    Child=(Type=Bone,Name="BASE_back_hair")
    Parent=(Type=Bone,Name="BASE_back_hair")
    Value=(Rotation=(X=0,Y=0,Z=0,W=1),Translation=(X=0,Y=0,Z=0),Scale3D=(X=0.8,Y=0.8,Z=0.8))
    bParentInitial=False
    bPropagateToChildren=True
    Weight=@n1.OutputParam
   ) {
    is=struct RigVMUnitNode
    template="Set Relative Transform::Execute(in Child,in Parent,in bParentInitial,in Value,in Weight,in bPropagateToChildren)"
   }
  ]}
]
hierarchy [
 root {loc=<0 0 0>, rot=<0 0 0 1>}
 pelvis {parent=root, loc=<0 -0.5301 98.0571>, rot=<0.0127 0.707 -0.0127 -0.707>}
 spine_01 {parent=pelvis, loc=<6.1208 -0.9328 0>, rot=<0 0 0.034 -0.9994>}
 spine_02 {parent=spine_01, loc=<6.5997 -0.2095 0>, rot=<0 0 0.0238 -0.9997>}
 spine_03 {parent=spine_02, loc=<6.4914 -0.1237 0>, rot=<0 0 0.0149 -0.9999>}
 spine_04 {parent=spine_03, loc=<6.9598 0.0794 0>, rot=<0 0 -0.0132 -0.9999>}
 spine_05 {parent=spine_04, loc=<6.9987 0 0>, rot=<0 0 -0.0015 -1>}
 neck_01 {parent=spine_05, loc=<11.3138 0.1529 0>, rot=<0 0 -0.2003 -0.9797>}
 neck_02 {parent=neck_01, loc=<4.5627 0 0>, rot=<0 0 0.0504 -0.9987>}
 head {parent=neck_02, loc=<5.3914 0 0>, rot=<0 0 0.115 -0.9934>}
 head_end {parent=head, loc=<9.2548 0 0>, rot=<0 0 0 1>}
 F_hair_01 {parent=head, loc=<7.4167 9.0306 1.0881>, rot=<-0.0916 -0.9958 0 0>}
 F_hair_02 {parent=F_hair_01, loc=<1.9623 0.1044 0.1378>, rot=<0 0 0.135 -0.9908>}
 BASE_pigtail_01 {parent=head, loc=<0.0482 5.5922 7.0788>, rot=<-0.0445 -0.9933 0.0562 -0.0909>}
 pigtail_01 {parent=BASE_pigtail_01, loc=<0 0 0>, rot=<0 0 0 1>}
 pigtail_02 {parent=pigtail_01, loc=<3.6776 0 0>, rot=<0 -0.0851 0 -0.9964>}
 pigtail_03 {parent=pigtail_02, loc=<3.3392 0 0>, rot=<-0.0385 0.0916 0.0035 -0.995>}
 pigtail_04 {parent=pigtail_03, loc=<3.0132 0 0>, rot=<-0.2012 -0.0793 -0.0164 -0.9762>}
 pigtail_05 {parent=pigtail_04, loc=<3.2096 0 0>, rot=<-0.2622 -0.1321 -0.0363 -0.9552>}
 pigtail_06 {parent=pigtail_05, loc=<3.4494 0 0>, rot=<0.1547 -0.082 0.0129 -0.9845>}
 pigtail_end {parent=pigtail_06, loc=<2.8482 0 0>, rot=<0 0 0 1>}
 L_hair_01 {parent=head, loc=<4.4873 5.9318 -6.4247>, rot=<-0.0039 -0.9999 -0.0002 -0.0145>}
 BASE_back_hair {parent=head, loc=<1.1193 -5.4514 0.0953>, rot=<0.365 -0.5677 -0.5653 0.4743>}
 back_hair_01 {parent=BASE_back_hair, loc=<1.5308 -0.9342 -3.1795>, rot=<0.0044 -0.3541 -0.103 -0.9295>}
 back_hair_02 {parent=back_hair_01, loc=<1.5097 -0.1764 0.0405>, rot=<0 0.1175 0 -0.9931>}
 back_hair_03 {parent=BASE_back_hair, loc=<3.99 -3.8588 1.4606>, rot=<0.9935 0.0739 -0.0639 -0.0581>}
 back_hair_04 {parent=BASE_back_hair, loc=<4.9037 1.8746 -1.9447>, rot=<0.9628 -0.0623 -0.2627 -0.0116>}
 back_hair_05 {parent=BASE_back_hair, loc=<1.9275 3.282 -1.0553>, rot=<-0.0505 -0.152 -0.3109 -0.9368>}
 clavicle_l {parent=spine_05, loc=<5.4831 -0.1053 -3.2946>, rot=<0.0465 -0.7289 -0.009 -0.6829>}
 upperarm_l {parent=clavicle_l, loc=<9.8658 -1.2136 -0.4105>, rot=<0.0164 -0.3245 -0.0269 -0.9454>}
 upperarm_twist_01_l {parent=upperarm_l, loc=<7.2135 -0.1134 0>, rot=<0 0 0 1>}
 deform_upperarm_bk_base_l {parent=upperarm_twist_01_l, loc=<-6.4053 -4.3746 0.314>, rot=<0 0 0 -1>}
 deform_upperarm_bk_l {parent=deform_upperarm_bk_base_l, loc=<0 0 0>, rot=<0 0 0 1>}
 deform_upperarm_up_base_l {parent=upperarm_twist_01_l, loc=<-8.891 0.4825 3.9517>, rot=<0 0 0 -1>}
 deform_upperarm_up_l {parent=deform_upperarm_up_base_l, loc=<0 0 0>, rot=<0 0 0 1>}
 clavicle_r {parent=spine_05, loc=<5.4831 -0.1053 3.2946>, rot=<0.7289 0.0465 -0.6829 0.009>}
 upperarm_r {parent=clavicle_r, loc=<-9.8658 1.2136 0.4105>, rot=<0.0164 -0.3245 -0.0269 -0.9454>}
 upperarm_twist_01_r {parent=upperarm_r, loc=<-7.2135 0.1134 0>, rot=<0 0 0 -1>}
 deform_upperarm_bk_base_r {parent=upperarm_twist_01_r, loc=<6.4053 4.3746 -0.314>, rot=<0 0 0 -1>}
 deform_upperarm_bk_r {parent=deform_upperarm_bk_base_r, loc=<0 0 0>, rot=<0 0 0 1>}
 deform_upperarm_fr_base_r {parent=upperarm_twist_01_r, loc=<5.6312 -5.092 -0.3354>, rot=<0 0 0 -1>}
 deform_upperarm_fr_r {parent=deform_upperarm_fr_base_r, loc=<0 0 0>, rot=<0 0 0 1>}
 deform_upperarm_up_base_r {parent=upperarm_twist_01_r, loc=<8.891 -0.4825 -3.9517>, rot=<0 0 0 -1>}
 deform_upperarm_up_r {parent=deform_upperarm_up_base_r, loc=<0 0 0>, rot=<0 0 0 1>}
 upperarm_bodies_01_r {parent=upperarm_r, loc=<-10.2129 0.4926 0.998>, rot=<0 0 0 -1>}
 upperarm_twist_02_r {parent=upperarm_r, loc=<-19.7998 0.1134 0>, rot=<0 0 0 -1>}
 BASE_collar_r4_r {parent=spine_05, loc=<10.0123 -0.3596 7.9517>, rot=<0.6143 0.1504 -0.03 -0.774>}
 collar_r4_r {parent=BASE_collar_r4_r, loc=<0 0 0>, rot=<0 0 0 1>}
 BASE_collar_l6_l {parent=spine_05, loc=<8.2461 4.0111 -6.8156>, rot=<0.2307 -0.8226 0.5135 -0.0798>}
 collar_l6_l {parent=BASE_collar_l6_l, loc=<0 0 0>, rot=<0 0 0 1>}
 BASE_collar_l7_l {parent=spine_05, loc=<7.1595 6.1923 -4.8084>, rot=<0.1782 -0.919 0.3393 -0.0929>}
 collar_l7_l {parent=BASE_collar_l7_l, loc=<0 0 0>, rot=<0 0 0 1>}
 BASE_collar_r7_r {parent=spine_05, loc=<7.1598 6.1924 4.8084>, rot=<0.938 0.1766 -0.0892 -0.2845>}
 collar_r7_r {parent=BASE_collar_r7_r, loc=<0 0 0>, rot=<0 0 0 1>}
 BASE_collar_r6_r {parent=spine_05, loc=<8.2461 4.0111 6.8156>, rot=<0.8226 0.2307 -0.0798 -0.5135>}
 collar_r6_r {parent=BASE_collar_r6_r, loc=<0 0 0>, rot=<0 0 0 1>}
 BASE_collar_r5_r {parent=spine_05, loc=<9.0661 1.6604 7.6292>, rot=<0.7645 0.1942 -0.0498 -0.6127>}
 collar_r5_r {parent=BASE_collar_r5_r, loc=<0 0 0>, rot=<0 0 0 1>}
 BASE_collar_m {parent=spine_05, loc=<9.991 -6.7419 0>, rot=<0.0002 -0.002 0.0198 -0.9998>}
 collar_m {parent=BASE_collar_m, loc=<0 0 0>, rot=<0 0 0 1>}
 BASE_collar_l5_l {parent=spine_05, loc=<9.0663 1.6604 -7.6292>, rot=<0.1942 -0.7645 0.6127 -0.0498>}
 collar_l5_l {parent=BASE_collar_l5_l, loc=<0 0 0>, rot=<0 0 0 1>}
 BASE_collar_l4_l {parent=spine_05, loc=<10.0123 -0.3596 -7.9517>, rot=<0.1478 -0.6144 0.7745 -0.0279>}
 collar_l4_l {parent=BASE_collar_l4_l, loc=<0 0 0>, rot=<0 0 0 1>}
 BASE_collar_l3_l {parent=spine_05, loc=<10.3773 -2.6667 -7.3731>, rot=<0.0901 -0.5721 0.8149 0.0217>}
 collar_l3_l {parent=BASE_collar_l3_l, loc=<0 0 0>, rot=<0 0 0 1>}
 BASE_collar_l2_l {parent=spine_05, loc=<10.1085 -4.6062 -5.8197>, rot=<0.04 -0.3145 0.9482 0.0218>}
 collar_l2_l {parent=BASE_collar_l2_l, loc=<0 0 0>, rot=<0 0 0 1>}
 BASE_collar_l1_l {parent=spine_05, loc=<10.0267 -6.2033 -3.0844>, rot=<0.0377 -0.187 0.9814 0.0199>}
 collar_l1_l {parent=BASE_collar_l1_l, loc=<0 0 0>, rot=<0 0 0 1>}
 BASE_collar_r1_r {parent=spine_05, loc=<10.0267 -6.2033 3.0844>, rot=<0.187 0.0377 0.0199 -0.9814>}
 collar_r1_r {parent=BASE_collar_r1_r, loc=<0 0 0>, rot=<0 0 0 1>}
 BASE_collar_r2_r {parent=spine_05, loc=<10.1085 -4.6062 5.8197>, rot=<0.3144 0.0365 0.023 -0.9483>}
 collar_r2_r {parent=BASE_collar_r2_r, loc=<0 0 0>, rot=<0 0 0 1>}
 BASE_collar_r3_r {parent=spine_05, loc=<10.3773 -2.6667 7.3731>, rot=<0.5226 0.0984 0.0119 -0.8468>}
 collar_r3_r {parent=BASE_collar_r3_r, loc=<0 0 0>, rot=<0 0 0 1>}
 deform_pec_aim_base_r {parent=spine_04, loc=<3.2452 9.3309 1.2237>, rot=<-0.4033 -0.0603 0.8507 0.3318>}
 deform_pec_aim_r {parent=deform_pec_aim_base_r, loc=<0 0 0>, rot=<0 0 0 1>}
 deform_pec_base_r {parent=deform_pec_aim_r, loc=<-9.3667 -0.0003 -0.0004>, rot=<0 0 0 1>}
 deform_pec_r {parent=deform_pec_base_r, loc=<0 0 0>, rot=<0 0 0 1>}
]
