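# Physics asset for the VanceGuard female medium body's face accessory mesh, covering hair motion and the upper-body shapes around it.
# A seven-bone pigtail chain and short front, left and back hair strands hang from the head as capsules with limited swing and twist.
# Head, spine, clavicles, upper arms and collar pieces carry larger shapes joined by angularly free constraints, apparently serving as colliders for the hair.
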PhysicsAsset /BRCosmetics/Characters/Player/Female/Medium/Bodies/F_MED_VanceGuard/Meshes/Parts/F_MED_VanceGuard_FaceAcc_Physics
bodies {
 F_hair_01 [{
   shapeType=Capsule
   length=1.675
   radius=1.0247
   shapeName=F_hair_01_capsule
   center=<1.2752 0.0022 -0.0529>
   rotation=<rot 8.9912 -90 2.3013>
  }]
 F_hair_02 [{
   shapeType=Capsule
   length=2.8794
   radius=1.4257
   shapeName=F_hair_02_capsule
   center=<1.7504 -0.2509 1.4652>
   rotation=<rot 0 80.5008 -50.252>
  }]
 L_hair_01 [{
   shapeType=Capsule
   length=2.0141
   radius=0.7372
   shapeName=L_hair_01_capsule
   center=<1.6044 -0.4438 0.1479>
   rotation=<rot 0 -118.0185 -88.6756>
  }]
 back_hair_01 [{
   shapeType=Capsule
   length=1.7455
   radius=1.536
   shapeName=back_hair_01_capsule
   center=<2.2336 -0.5262 0.035>
   rotation=<rot 0 70.8072 -86.3263>
  }]
 back_hair_02 [{
   shapeType=Capsule
   length=1.8283
   radius=1.6766
   shapeName=back_hair_02_capsule
   center=<2.3662 0.211 0.809>
   rotation=<rot 0 98.1342 -66.9537>
  }]
 back_hair_03 [{
   shapeType=Capsule
   length=1.8929
   radius=1.1805
   shapeName=back_hair_03_capsule
   center=<2.2841 0.0223 -0.1441>
   rotation=<rot 0 -89.7798 -87.3037>
  }]
 back_hair_04 [{
   shapeType=Capsule
   length=2.2625
   radius=1.2785
   shapeName=back_hair_04_capsule
   center=<2.1181 0.1483 0.2292>
   rotation=<rot 0 101.828 -84.7659>
  }]
 back_hair_05 [{
   shapeType=Capsule
   length=1.8891
   radius=1.1273
   shapeName=back_hair_05_capsule
   center=<1.8491 -0.0087 0.5304>
   rotation=<rot -3.9587 89.2254 -69.3297>
  }]
 clavicle_l [{
   shapeType=Capsule
   length=4.4074
   radius=8.2839
   shapeName=clavicle_l_capsule
   center=<5.2995 0 -3.0031>
   rotation=<rot -0.0001 -90 -88.2704>
  }]
 clavicle_r [{
   shapeType=Capsule
   length=5.6987
   radius=6.9926
   shapeName=clavicle_r_capsule
   center=<-5.3479 2.1354 1.4003>
   rotation=<rot -0.0001 -90 91.7296>
  }]
 collar_l6_l [{shapeType=Capsule, length=1.2961, radius=4.0426, shapeName=collar_l7_l_capsule, center=<-2.9492 -3.9469 0>, rotation=<rot 0 -90 0>}]
 collar_l7_l [{shapeType=Capsule, length=1.2961, radius=4.0426, shapeName=collar_l7_l_capsule, center=<-2.9492 -3.9469 0>, rotation=<rot 0 -90 0>}]
 collar_m [{shapeType=Capsule, length=1.2961, radius=4.0426, shapeName=collar_l7_l_capsule, center=<5.2773 3.0699 0>, rotation=<rot 0 -90 0>}]
 collar_r1_r [{shapeType=Capsule, length=1.2961, radius=4.0426, shapeName=collar_l7_l_capsule, center=<4.8004 2.7834 0>, rotation=<rot 0 -90 0>}]
 collar_r2_r [{shapeType=Capsule, length=1.2961, radius=4.0426, shapeName=collar_l7_l_capsule, center=<3.9927 2.4399 0>, rotation=<rot 0 -90 0>}]
 collar_r3_r [{shapeType=Capsule, length=1.2961, radius=4.0426, shapeName=collar_l7_l_capsule, center=<3.2861 3.2703 0>, rotation=<rot 0 -90 0>}]
 collar_r4_r [{shapeType=Capsule, length=1.2961, radius=4.0426, shapeName=collar_l7_l_capsule, center=<2.901 3.5325 0>, rotation=<rot 0 -90 0>}]
 collar_r5_r [{shapeType=Capsule, length=1.2961, radius=4.0426, shapeName=collar_l7_l_capsule, center=<2.4758 3.9123 0>, rotation=<rot 0 -90 0>}]
 collar_r6_r [{shapeType=Capsule, length=1.2961, radius=4.0426, shapeName=collar_l7_l_capsule, center=<2.4758 3.9123 0>, rotation=<rot 0 -90 0>}]
 collar_r7_r [{shapeType=Capsule, length=1.2961, radius=4.0426, shapeName=collar_l7_l_capsule, center=<2.4758 3.9123 0>, rotation=<rot 0 -90 0>}]
 head [
  {shapeType=Sphere, radius=8.5694, shapeName=head_sphere, center=<4.5428 0 0>}
  {shapeType=Capsule, length=7.7534, radius=5.3369, center=<3.1734 2.6019 0.0541>, rotation=<rot 0.5016 -90.415 -88.3249>}
  {
   shapeType=Capsule
   length=2.7534
   radius=3.8369
   shapeName=head_capsule
   center=<1.1072 -0.3919 0.4333>
   rotation=<rot 0.5016 -90.415 -88.3249>
  }
  {
   shapeType=Capsule
   length=7.7534
   radius=5.3369
   shapeName=head_capsule
   center=<6.4152 0.0461 3.5786>
   rotation=<rot 0.5016 -90.415 -88.3249>
  }
 ]
 pigtail_01 [{
   shapeType=Capsule
   length=2.5134
   radius=1.0434
   shapeName=pigtail_01_capsule
   center=<2.3756 0 -0.0427>
   rotation=<rot -88.9713 179.9991 -179.999>
  }]
 pigtail_02 [{
   shapeType=Capsule
   length=2.5134
   radius=1.0434
   shapeName=pigtail_02_capsule
   center=<2.3328 0 -0.0419>
   rotation=<rot -88.9713 -180 -179.9999>
  }]
 pigtail_03 [{
   shapeType=Capsule
   length=2.5134
   radius=1.0434
   shapeName=pigtail_03_capsule
   center=<2.3635 0 -0.0424>
   rotation=<rot -88.9713 179.999 -179.999>
  }]
 pigtail_04 [{
   shapeType=Capsule
   length=2.5134
   radius=1.0434
   shapeName=pigtail_04_capsule
   center=<2.0787 0 -0.0373>
   rotation=<rot -88.9712 179.9997 -179.9997>
  }]
 pigtail_05 [{
   shapeType=Capsule
   length=2.5134
   radius=1.0434
   shapeName=pigtail_05_capsule
   center=<2.4467 0 -0.0439>
   rotation=<rot -88.9712 -179.9996 179.9997>
  }]
 pigtail_06 [{
   shapeType=Capsule
   length=2.5134
   radius=1.0434
   shapeName=pigtail_06_capsule
   center=<2.2986 0 -0.0413>
   rotation=<rot -88.9712 -179.999 179.9991>
  }]
 pigtail_end [{
   shapeType=Capsule
   length=2.3516
   radius=1.2052
   shapeName=pigtail_end_capsule
   center=<2.4346 0.0941 -0.2984>
   rotation=<rot -61.6384 113.9719 -111.8312>
  }]
 spine_04 [{shapeType=Capsule, length=5.1813, radius=13.3746, shapeName=spine_04_capsule, center=<0 -0.48 0>, rotation=<rot 0 -90 0>}]
 spine_05 [
  {shapeType=Capsule, length=11.4867, radius=9.9217, shapeName=spine_05_capsule, center=<0 -0.8547 0>, rotation=<rot 0 -90 0>}
  {shapeType=Capsule, length=2.8674, radius=6.0752, shapeName=spine_05_capsule, center=<3.4398 1.1631 13.4089>, rotation=<rot 0 -90 0>}
 ]
 upperarm_bodies_01_r [{shapeType=Capsule, length=16.8864, radius=5.4329, shapeName=upperarm_bodies_01_r_capsule, rotation=<rot 0 -90 90>}]
 upperarm_l [{
   shapeType=Capsule
   length=16.7998
   radius=6.1331
   shapeName=upperarm_l_capsule
   center=<9.8257 0 -1.1851>
   rotation=<rot -0.0001 -89.9999 87.131>
  }]
 upperarm_r [{
   shapeType=Capsule
   length=26.2554
   radius=4.9549
   shapeName=upperarm_r_capsule
   center=<-13.6288 2.4377 0.9946>
   rotation=<rot -0.0001 -89.9999 -92.0781>
  }]
}
constraints [
 {bone1Name=pigtail_01, bone2Name=head, swing1LimitDegrees=50, swing2LimitDegrees=50, twistLimitDegrees=5}
 {bone1Name=pigtail_02, bone2Name=pigtail_01, swing1LimitDegrees=50, swing2LimitDegrees=50, twistLimitDegrees=5}
 {bone1Name=pigtail_03, bone2Name=pigtail_02, swing1LimitDegrees=50, swing2LimitDegrees=50, twistLimitDegrees=5}
 {bone1Name=pigtail_04, bone2Name=pigtail_03, swing1LimitDegrees=50, swing2LimitDegrees=50, twistLimitDegrees=5}
 {bone1Name=pigtail_05, bone2Name=pigtail_04, swing1LimitDegrees=50, swing2LimitDegrees=50, twistLimitDegrees=5}
 {bone1Name=pigtail_06, bone2Name=pigtail_05, swing1LimitDegrees=50, swing2LimitDegrees=50, twistLimitDegrees=5}
 {bone1Name=pigtail_end, bone2Name=pigtail_06, swing1LimitDegrees=50, swing2LimitDegrees=50, twistLimitDegrees=5}
 {bone1Name=F_hair_01, bone2Name=head, swing1LimitDegrees=20, swing2LimitDegrees=15, twistLimitDegrees=5}
 {bone1Name=F_hair_02, bone2Name=F_hair_01, swing1LimitDegrees=20, swing2LimitDegrees=15, twistLimitDegrees=5}
 {bone1Name=L_hair_01, bone2Name=head, swing1LimitDegrees=25, swing2LimitDegrees=25, twistLimitDegrees=5}
 {bone1Name=back_hair_01, bone2Name=head, swing1LimitDegrees=30, swing2LimitDegrees=30, twistLimitDegrees=5}
 {bone1Name=back_hair_02, bone2Name=back_hair_01, swing1LimitDegrees=30, swing2LimitDegrees=30, twistLimitDegrees=5}
 {bone1Name=back_hair_03, bone2Name=head, swing1LimitDegrees=25, swing2LimitDegrees=25, twistLimitDegrees=5}
 {bone1Name=back_hair_04, bone2Name=head, swing1LimitDegrees=20, swing2LimitDegrees=20, twistLimitDegrees=5}
 {bone1Name=back_hair_05, bone2Name=head, swing1LimitDegrees=45, swing2LimitDegrees=45, twistLimitDegrees=45}
 {
  bone1Name=upperarm_l
  bone2Name=clavicle_l
  swing1LimitDegrees=45
  swing1Motion=Free
  swing2LimitDegrees=45
  swing2Motion=Free
  twistLimitDegrees=45
  twistMotion=Free
 }
 {
  bone1Name=upperarm_r
  bone2Name=clavicle_r
  swing1LimitDegrees=45
  swing1Motion=Free
  swing2LimitDegrees=45
  swing2Motion=Free
  twistLimitDegrees=45
  twistMotion=Free
 }
 {
  bone1Name=head
  bone2Name=spine_05
  swing1LimitDegrees=45
  swing1Motion=Free
  swing2LimitDegrees=45
  swing2Motion=Free
  twistLimitDegrees=45
  twistMotion=Free
 }
 {
  bone1Name=clavicle_l
  bone2Name=spine_05
  swing1LimitDegrees=45
  swing1Motion=Free
  swing2LimitDegrees=45
  swing2Motion=Free
  twistLimitDegrees=45
  twistMotion=Free
 }
 {
  bone1Name=clavicle_r
  bone2Name=spine_05
  swing1LimitDegrees=45
  swing1Motion=Free
  swing2LimitDegrees=45
  swing2Motion=Free
  twistLimitDegrees=45
  twistMotion=Free
 }
 {
  bone1Name=collar_r4_r
  bone2Name=spine_05
  swing1LimitDegrees=45
  swing1Motion=Free
  swing2LimitDegrees=45
  swing2Motion=Free
  twistLimitDegrees=45
  twistMotion=Free
 }
 {
  bone1Name=collar_l6_l
  bone2Name=spine_05
  swing1LimitDegrees=45
  swing1Motion=Free
  swing2LimitDegrees=45
  swing2Motion=Free
  twistLimitDegrees=45
  twistMotion=Free
 }
 {
  bone1Name=collar_l7_l
  bone2Name=spine_05
  swing1LimitDegrees=45
  swing1Motion=Free
  swing2LimitDegrees=45
  swing2Motion=Free
  twistLimitDegrees=45
  twistMotion=Free
 }
 {
  bone1Name=collar_r7_r
  bone2Name=spine_05
  swing1LimitDegrees=45
  swing1Motion=Free
  swing2LimitDegrees=45
  swing2Motion=Free
  twistLimitDegrees=45
  twistMotion=Free
 }
 {
  bone1Name=collar_r6_r
  bone2Name=spine_05
  swing1LimitDegrees=45
  swing1Motion=Free
  swing2LimitDegrees=45
  swing2Motion=Free
  twistLimitDegrees=45
  twistMotion=Free
 }
 {
  bone1Name=collar_r5_r
  bone2Name=spine_05
  swing1LimitDegrees=45
  swing1Motion=Free
  swing2LimitDegrees=45
  swing2Motion=Free
  twistLimitDegrees=45
  twistMotion=Free
 }
 {
  bone1Name=spine_05
  bone2Name=spine_04
  swing1LimitDegrees=45
  swing1Motion=Free
  swing2LimitDegrees=45
  swing2Motion=Free
  twistLimitDegrees=45
  twistMotion=Free
 }
 {
  bone1Name=upperarm_bodies_01_r
  bone2Name=upperarm_r
  swing1LimitDegrees=45
  swing1Motion=Free
  swing2LimitDegrees=45
  swing2Motion=Free
  twistLimitDegrees=45
  twistMotion=Free
 }
 {
  bone1Name=collar_r3_r
  bone2Name=spine_05
  swing1LimitDegrees=45
  swing1Motion=Free
  swing2LimitDegrees=45
  swing2Motion=Free
  twistLimitDegrees=45
  twistMotion=Free
 }
 {
  bone1Name=collar_r2_r
  bone2Name=spine_05
  swing1LimitDegrees=45
  swing1Motion=Free
  swing2LimitDegrees=45
  swing2Motion=Free
  twistLimitDegrees=45
  twistMotion=Free
 }
 {
  bone1Name=collar_r1_r
  bone2Name=spine_05
  swing1LimitDegrees=45
  swing1Motion=Free
  swing2LimitDegrees=45
  swing2Motion=Free
  twistLimitDegrees=45
  twistMotion=Free
 }
 {
  bone1Name=collar_m
  bone2Name=spine_05
  swing1LimitDegrees=45
  swing1Motion=Free
  swing2LimitDegrees=45
  swing2Motion=Free
  twistLimitDegrees=45
  twistMotion=Free
 }
]
